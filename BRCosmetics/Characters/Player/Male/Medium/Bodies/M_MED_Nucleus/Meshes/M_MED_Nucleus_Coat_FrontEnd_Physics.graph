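# Physics asset for the front-end variant of the M_MED_Nucleus male character's coat, defining bodies for jacket dynamics.
# Six five-segment capsule chains (front, side and back, left and right) hang from tiny master bodies on the pelvis, with very tight twist limits and narrower swing on the back.
# Root, pelvis, leg and holster bodies complete the setup, joined by free angular constraints, as are right side and front segment 02.
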PhysicsAsset /BRCosmetics/Characters/Player/Male/Medium/Bodies/M_MED_Nucleus/Meshes/M_MED_Nucleus_Coat_FrontEnd_Physics
bodies {
 calf_l [{shapeType=Capsule, length=45.8142, radius=9.5, center=<-26.378 -3.757 0.0142>, rotation=<rot 90 -90 -90.1835>}]
 calf_r [{shapeType=Capsule, length=42.2071, radius=11.5191, center=<27.1517 4.7196 0.2975>, rotation=<rot 90 0 5.0002>}]
 dyn_holster [{
   shapeType=Capsule
   length=17.9376
   radius=9.8623
   shapeName=dyn_holster_capsule
   center=<6.0902 -1.1194 3.5502>
   rotation=<rot 2.7131 -91.1304 -70.8281>
  }]
 dyn_jacketBack_01_l [{shapeType=Capsule, length=17.6418, radius=11.5, shapeName=dyn_jacketBack_01_l_capsule, center=<5 0 10>, rotation=<rot 0 0 90>}]
 dyn_jacketBack_01_r [{shapeType=Capsule, length=16.4586, radius=12, shapeName=dyn_jacketBack_01_r_capsule, center=<5 0 10>, rotation=<rot 0 0 90>}]
 dyn_jacketBack_02_l [{shapeType=Capsule, length=16, radius=12.5, shapeName=dyn_jacketBack_02_l_capsule, center=<6 0 10>, rotation=<rot 0 0 90>}]
 dyn_jacketBack_02_r [{shapeType=Capsule, length=13, radius=14.5, shapeName=dyn_jacketBack_02_r_capsule, center=<6 0 10>, rotation=<rot 0 0 90>}]
 dyn_jacketBack_03_l [{shapeType=Capsule, length=18, radius=11.5, shapeName=dyn_jacketBack_03_l_capsule, center=<2 0 10>, rotation=<rot 0 0 90>}]
 dyn_jacketBack_03_r [{shapeType=Capsule, length=21.5, radius=15, shapeName=dyn_jacketBack_03_r_capsule, center=<2 0 10>, rotation=<rot 0 0 90>}]
 dyn_jacketBack_04_l [{shapeType=Capsule, length=28.5, radius=9.5, shapeName=dyn_jacketBack_04_l_capsule, center=<2 0 10>, rotation=<rot 0 0 90>}]
 dyn_jacketBack_04_r [{shapeType=Capsule, length=20.25, radius=14.5, shapeName=dyn_jacketBack_04_l_capsule, center=<2 0 10>, rotation=<rot 0 0 90>}]
 dyn_jacketBack_05_l [{shapeType=Capsule, length=31.5, radius=9.5, shapeName=dyn_jacketBack_05_l_capsule, center=<9.6914 0 10>, rotation=<rot 0 0 90>}]
 dyn_jacketBack_05_r [{shapeType=Capsule, length=18.25, radius=15.5, shapeName=dyn_jacketBack_05_r_capsule, center=<2 0 10>, rotation=<rot 0 0 90>}]
 dyn_jacketBack_mstr_l [{shapeType=Capsule, length=0.505, radius=0.505, rotation=<rot 0 -90 0>}]
 dyn_jacketBack_mstr_r [{shapeType=Capsule, length=0.505, radius=0.505, rotation=<rot 0 -90 0>}]
 dyn_jacketFront_01_l [{shapeType=Capsule, length=15, radius=4, shapeName=dyn_jacketFront_01_l_capsule, center=<5 5 0>, rotation=<rot 90 0 0>}]
 dyn_jacketFront_01_r [{shapeType=Capsule, length=11, radius=4, shapeName=dyn_jacketFront_01_r_capsule, center=<8.0011 0.0001 0.4904>, rotation=<rot 90 0 0>}]
 dyn_jacketFront_02_l [{shapeType=Capsule, length=22.5, radius=5, shapeName=dyn_jacketFront_02_l_capsule, center=<9 0.0001 2.4>, rotation=<rot 0 0 89.9999>}]
 dyn_jacketFront_02_r [{shapeType=Capsule, length=21.3305, radius=5, shapeName=dyn_jacketFront_02_r_capsule, center=<9.7618 -1.5 2.2>, rotation=<rot 0 0 60>}]
 dyn_jacketFront_03_l [
  {
   shapeType=Capsule
   length=22.3617
   radius=8
   shapeName=dyn_jacketFront_03_l_capsule
   center=<9.9238 0.0001 -0.0001>
   rotation=<rot 0 0 -81.5266>
  }
  {
   shapeType=Capsule
   length=16
   radius=1.25
   shapeName=dyn_jacketFront_03_l_capsule
   center=<9.5898 0 0.5923>
   rotation=<rot 86.4659 179.9998 179.9998>
  }
 ]
 dyn_jacketFront_03_r [{shapeType=Capsule, length=16.5248, radius=7, shapeName=dyn_jacketFront_03_l_capsule, center=<9.7618 -1.6 1.8>, rotation=<rot 0 0 70>}]
 dyn_jacketFront_04_l [{shapeType=Capsule, length=23.599, radius=10, shapeName=dyn_jacketFront_04_l_capsule, center=<8.5574 0 1>, rotation=<rot 0 0 -89.9999>}]
 dyn_jacketFront_04_r [{shapeType=Capsule, length=21.5, radius=7, shapeName=dyn_jacketFront_04_r_capsule, center=<5.3986 0 0>, rotation=<rot 0 0 80>}]
 dyn_jacketFront_05_l [{shapeType=Capsule, length=18.25, radius=10, shapeName=dyn_jacketFront_05_l_capsule, center=<7.3524 0.0001 1>, rotation=<rot 0 0 -89.9999>}]
 dyn_jacketFront_05_r [{
   shapeType=Capsule
   length=27
   radius=7
   shapeName=dyn_jacketFront_05_l_capsule
   center=<4.8209 0.4017 0.2777>
   rotation=<rot 5.569 8.1026 80>
  }]
 dyn_jacketFront_mstr_l [{shapeType=Capsule, length=0.505, radius=0.505, rotation=<rot 0 -90 0>}]
 dyn_jacketFront_mstr_r [{shapeType=Capsule, length=0.505, radius=0.505, rotation=<rot 0 -90 0>}]
 dyn_jacketSide_01_l [{shapeType=Capsule, length=14, radius=4, shapeName=dyn_jacketSide_01_l_capsule, center=<5 0 3>, rotation=<rot 0 0 75.4999>}]
 dyn_jacketSide_01_r [{shapeType=Capsule, length=9.5, radius=8.5, shapeName=dyn_jacketSide_01_l_capsule, center=<5 -2 9>, rotation=<rot 0 0 75.4999>}]
 dyn_jacketSide_02_l [{shapeType=Capsule, length=18.75, radius=5, shapeName=dyn_jacketSide_02_l_capsule, center=<9 -0.2195 3>, rotation=<rot 0 0 71.9271>}]
 dyn_jacketSide_02_r [{shapeType=Capsule, length=16.5248, radius=9.5, shapeName=dyn_jacketSide_02_l_capsule, center=<9.7618 0 10>, rotation=<rot 0 0 89.9999>}]
 dyn_jacketSide_03_l [{shapeType=Capsule, length=22.75, radius=8, shapeName=dyn_jacketSide_03_l_capsule, center=<3.2696 0 3>, rotation=<rot 0 0 71.0282>}]
 dyn_jacketSide_03_r [{shapeType=Capsule, length=17.7748, radius=8, shapeName=dyn_jacketSide_03_l_capsule, center=<9.7618 0.0001 3>, rotation=<rot 0 0 89.9999>}]
 dyn_jacketSide_04_l [{shapeType=Capsule, length=26.75, radius=8, shapeName=dyn_jacketSide_04_l_capsule, center=<6 -3.1656 1.1034>, rotation=<rot 0 0 59.0735>}]
 dyn_jacketSide_04_r [{shapeType=Capsule, length=26.25, radius=8, shapeName=dyn_jacketSide_04_r_capsule, center=<7.2199 0 3>, rotation=<rot 0 0 90>}]
 dyn_jacketSide_05_l [{shapeType=Capsule, length=31.75, radius=8, shapeName=dyn_jacketSide_05_l_capsule, center=<8.4738 -3.0947 0.9899>, rotation=<rot 0 0 56.9951>}]
 dyn_jacketSide_05_r [{shapeType=Capsule, length=37.25, radius=8, shapeName=dyn_jacketSide_05_r_capsule, center=<7.4186 0 3>, rotation=<rot 0 0 105.0003>}]
 dyn_jacketSide_mstr_l [{shapeType=Capsule, length=0.505, radius=0.505, rotation=<rot 0 -90 0>}]
 dyn_jacketSide_mstr_r [{shapeType=Capsule, length=0.505, radius=0.505, rotation=<rot 0 -90 0>}]
 pelvis [
  {shapeType=Capsule, length=11.25, radius=13.75, center=<-0.649 0.2801 -0.0002>, rotation=<rot 90 0 -3.3073>}
  {
   shapeType=Capsule
   length=15.3785
   radius=7.9107
   shapeName=pelvis_capsule
   center=<-0.0808 -1.3903 -4.5558>
   rotation=<rot 80.001 -176.6924 -180>
  }
  {shapeType=Capsule, length=15.6212, radius=7.6681, shapeName=pelvis_capsule, center=<-0.2328 -1.3991 5.3434>, rotation=<rot 77.8519 3.3074 0>}
 ]
 root [{shapeType=Sphere, radius=17.38, shapeName=root_sphere, center=<0.0002 -3.221 13.6507>}]
 thigh_l [{shapeType=Capsule, length=30.25, radius=9.247, center=<-21.4597 0.1549 0.7921>, rotation=<rot 90 12.9074 10.9407>}]
 thigh_r [{shapeType=Capsule, length=25.1879, radius=10.0454, center=<21.7868 0 0>, rotation=<rot 90 0 0>}]
}
constraints [
 {
  bone1Name=thigh_l
  bone2Name=pelvis
  swing1LimitDegrees=45
  swing1Motion=Free
  swing2LimitDegrees=45
  swing2Motion=Free
  twistLimitDegrees=45
  twistMotion=Free
 }
 {
  bone1Name=calf_l
  bone2Name=thigh_l
  swing1LimitDegrees=45
  swing1Motion=Free
  swing2LimitDegrees=45
  swing2Motion=Free
  twistLimitDegrees=45
  twistMotion=Free
 }
 {
  bone1Name=thigh_r
  bone2Name=pelvis
  swing1LimitDegrees=45
  swing1Motion=Free
  swing2LimitDegrees=45
  swing2Motion=Free
  twistLimitDegrees=45
  twistMotion=Free
 }
 {
  bone1Name=calf_r
  bone2Name=thigh_r
  swing1LimitDegrees=45
  swing1Motion=Free
  swing2LimitDegrees=45
  swing2Motion=Free
  twistLimitDegrees=45
  twistMotion=Free
 }
 {
  bone1Name=dyn_jacketSide_mstr_l
  bone2Name=pelvis
  swing1LimitDegrees=45
  swing1Motion=Free
  swing2LimitDegrees=45
  swing2Motion=Free
  twistLimitDegrees=45
  twistMotion=Free
 }
 {
  bone1Name=dyn_jacketBack_mstr_l
  bone2Name=pelvis
  swing1LimitDegrees=45
  swing1Motion=Free
  swing2LimitDegrees=45
  swing2Motion=Free
  twistLimitDegrees=45
  twistMotion=Free
 }
 {
  bone1Name=dyn_jacketBack_mstr_r
  bone2Name=pelvis
  swing1LimitDegrees=45
  swing1Motion=Free
  swing2LimitDegrees=45
  swing2Motion=Free
  twistLimitDegrees=45
  twistMotion=Free
 }
 {
  bone1Name=dyn_jacketSide_mstr_r
  bone2Name=pelvis
  swing1LimitDegrees=45
  swing1Motion=Free
  swing2LimitDegrees=45
  swing2Motion=Free
  twistLimitDegrees=45
  twistMotion=Free
 }
 {
  bone1Name=dyn_jacketFront_mstr_l
  bone2Name=pelvis
  swing1LimitDegrees=45
  swing1Motion=Free
  swing2LimitDegrees=45
  swing2Motion=Free
  twistLimitDegrees=45
  twistMotion=Free
 }
 {
  bone1Name=dyn_jacketFront_mstr_r
  bone2Name=pelvis
  swing1LimitDegrees=45
  swing1Motion=Free
  swing2LimitDegrees=45
  swing2Motion=Free
  twistLimitDegrees=45
  twistMotion=Free
 }
 {bone1Name=dyn_jacketFront_01_r, bone2Name=dyn_jacketFront_mstr_r, swing1LimitDegrees=45, swing2LimitDegrees=45, twistLimitDegrees=0.5}
 {bone1Name=dyn_jacketFront_02_r, bone2Name=dyn_jacketFront_01_r, swing1LimitDegrees=45, swing2LimitDegrees=45, twistLimitDegrees=0.5}
 {bone1Name=dyn_jacketFront_03_r, bone2Name=dyn_jacketFront_02_r, swing1LimitDegrees=45, swing2LimitDegrees=45, twistLimitDegrees=0.5}
 {bone1Name=dyn_jacketFront_04_r, bone2Name=dyn_jacketFront_03_r, swing1LimitDegrees=45, swing2LimitDegrees=45, twistLimitDegrees=0.5}
 {bone1Name=dyn_jacketFront_01_l, bone2Name=dyn_jacketFront_mstr_l, swing1LimitDegrees=20, swing2LimitDegrees=45, twistLimitDegrees=0.5}
 {bone1Name=dyn_jacketFront_02_l, bone2Name=dyn_jacketFront_01_l, swing1LimitDegrees=45, swing2LimitDegrees=45, twistLimitDegrees=0.5}
 {bone1Name=dyn_jacketFront_03_l, bone2Name=dyn_jacketFront_02_l, swing1LimitDegrees=45, swing2LimitDegrees=45, twistLimitDegrees=0.5}
 {bone1Name=dyn_jacketFront_04_l, bone2Name=dyn_jacketFront_03_l, swing1LimitDegrees=45, swing2LimitDegrees=45, twistLimitDegrees=0.5}
 {bone1Name=dyn_jacketSide_01_r, bone2Name=dyn_jacketSide_mstr_r, swing1LimitDegrees=45, swing2LimitDegrees=45, twistLimitDegrees=0.5}
 {bone1Name=dyn_jacketSide_02_r, bone2Name=dyn_jacketSide_01_r, swing1LimitDegrees=45, swing2LimitDegrees=45, twistLimitDegrees=0.5}
 {bone1Name=dyn_jacketSide_03_r, bone2Name=dyn_jacketSide_02_r, swing1LimitDegrees=45, swing2LimitDegrees=45, twistLimitDegrees=0.5}
 {bone1Name=dyn_jacketSide_04_r, bone2Name=dyn_jacketSide_03_r, swing1LimitDegrees=45, swing2LimitDegrees=45, twistLimitDegrees=0.5}
 {bone1Name=dyn_jacketBack_01_r, bone2Name=dyn_jacketBack_mstr_r, swing1LimitDegrees=20, swing2LimitDegrees=20, twistLimitDegrees=0.5}
 {bone1Name=dyn_jacketBack_02_r, bone2Name=dyn_jacketBack_01_r, swing1LimitDegrees=20, swing2LimitDegrees=20, twistLimitDegrees=0.5}
 {bone1Name=dyn_jacketBack_03_r, bone2Name=dyn_jacketBack_02_r, swing1LimitDegrees=20, swing2LimitDegrees=20, twistLimitDegrees=1}
 {bone1Name=dyn_jacketBack_04_r, bone2Name=dyn_jacketBack_03_r, swing1LimitDegrees=20, swing2LimitDegrees=20, twistLimitDegrees=1}
 {bone1Name=dyn_jacketBack_01_l, bone2Name=dyn_jacketBack_mstr_l, swing1LimitDegrees=20, swing2LimitDegrees=20, twistLimitDegrees=0.5}
 {bone1Name=dyn_jacketBack_02_l, bone2Name=dyn_jacketBack_01_l, swing1LimitDegrees=20, swing2LimitDegrees=20, twistLimitDegrees=1}
 {bone1Name=dyn_jacketBack_03_l, bone2Name=dyn_jacketBack_02_l, swing1LimitDegrees=20, swing2LimitDegrees=20, twistLimitDegrees=1}
 {bone1Name=dyn_jacketBack_04_l, bone2Name=dyn_jacketBack_03_l, swing1LimitDegrees=20, swing2LimitDegrees=20, twistLimitDegrees=1}
 {bone1Name=dyn_jacketSide_01_l, bone2Name=dyn_jacketSide_mstr_l, swing1LimitDegrees=12, swing2LimitDegrees=40, twistLimitDegrees=0.5}
 {bone1Name=dyn_jacketSide_02_l, bone2Name=dyn_jacketSide_01_l, swing1LimitDegrees=45, swing2LimitDegrees=45, twistLimitDegrees=0.5}
 {bone1Name=dyn_jacketSide_03_l, bone2Name=dyn_jacketSide_02_l, swing1LimitDegrees=45, swing2LimitDegrees=45, twistLimitDegrees=1}
 {bone1Name=dyn_jacketSide_04_l, bone2Name=dyn_jacketSide_03_l, swing1LimitDegrees=45, swing2LimitDegrees=45, twistLimitDegrees=1}
 {
  bone1Name=pelvis
  bone2Name=root
  swing1LimitDegrees=45
  swing1Motion=Free
  swing2LimitDegrees=45
  swing2Motion=Free
  twistLimitDegrees=45
  twistMotion=Free
 }
 {bone1Name=dyn_jacketFront_05_r, bone2Name=dyn_jacketFront_04_r, swing1LimitDegrees=45, swing2LimitDegrees=45, twistLimitDegrees=0.5}
 {bone1Name=dyn_jacketFront_05_l, bone2Name=dyn_jacketFront_04_l, swing1LimitDegrees=45, swing2LimitDegrees=45, twistLimitDegrees=0.5}
 {bone1Name=dyn_jacketSide_05_r, bone2Name=dyn_jacketSide_04_r, swing1LimitDegrees=45, swing2LimitDegrees=45, twistLimitDegrees=0.5}
 {bone1Name=dyn_jacketSide_05_l, bone2Name=dyn_jacketSide_04_l, swing1LimitDegrees=45, swing2LimitDegrees=45, twistLimitDegrees=1}
 {bone1Name=dyn_jacketBack_05_l, bone2Name=dyn_jacketBack_04_l, swing1LimitDegrees=20, swing2LimitDegrees=20, twistLimitDegrees=1}
 {bone1Name=dyn_jacketBack_05_r, bone2Name=dyn_jacketBack_04_r, swing1LimitDegrees=20, swing2LimitDegrees=20, twistLimitDegrees=1}
 {
  bone1Name=dyn_holster
  bone2Name=thigh_r
  swing1LimitDegrees=45
  swing1Motion=Free
  swing2LimitDegrees=45
  swing2Motion=Free
  twistLimitDegrees=45
  twistMotion=Free
 }
 {
  bone1Name=dyn_jacketSide_02_r
  bone2Name=dyn_jacketFront_02_r
  swing1LimitDegrees=45
  swing1Motion=Free
  swing2LimitDegrees=45
  swing2Motion=Free
  twistLimitDegrees=45
  twistMotion=Free
 }
]
