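# Physics asset for the coat of the M_MED_Nucleus player body, providing bodies and joints for the jacket's dynamic bones.
# Six five-segment capsule chains (front, side, back per side) descend from master bones on the pelvis through swing- and twist-limited joints.
# Angular-free constraints cross-link the upper three front and side segments.
# Pelvis, leg and holster capsules, a root sphere and a large offset ground sphere add body and ground shapes.
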PhysicsAsset /BRCosmetics/Characters/Player/Male/Medium/Bodies/M_MED_Nucleus/Meshes/M_MED_Nucleus_Coat_Physics
bodies {
 calf_l [{shapeType=Capsule, length=49.9903, radius=10.1302, center=<-31.3035 -0.2388 0.0141>, rotation=<rot 90 0 -0.1835>}]
 calf_r [{shapeType=Capsule, length=47.6107, radius=11.5519, center=<32.5574 0 0.0001>, rotation=<rot 90 0 0>}]
 dyn_ground_root [{shapeType=Sphere, radius=300, shapeName=dyn_root_slope_sphere, center=<0 -20 -295>}]
 dyn_holster [{
   shapeType=Capsule
   length=17.8688
   radius=10.1531
   shapeName=dyn_holster_capsule
   center=<9.6892 -1.8492 1.5367>
   rotation=<rot 2.7131 -91.1305 -71.2677>
  }]
 dyn_jacketBack_01_l [{shapeType=Capsule, length=19.6418, radius=9.5, shapeName=dyn_jacketBack_01_l_capsule, center=<5 0 10>, rotation=<rot 0 0 90>}]
 dyn_jacketBack_01_r [{shapeType=Capsule, length=18.9586, radius=9.5, shapeName=dyn_jacketBack_01_r_capsule, center=<5 0 10>, rotation=<rot 0 0 90>}]
 dyn_jacketBack_02_l [{shapeType=Capsule, length=18, radius=10.5, shapeName=dyn_jacketBack_02_l_capsule, center=<6 0 10>, rotation=<rot 90 0 0>}]
 dyn_jacketBack_02_r [{shapeType=Capsule, length=17, radius=11, shapeName=dyn_jacketBack_02_r_capsule, center=<6 0 10>, rotation=<rot 90 0 0>}]
 dyn_jacketBack_03_l [{shapeType=Capsule, length=20, radius=9.5, shapeName=dyn_jacketBack_03_l_capsule, center=<2 0 10>, rotation=<rot 0 0 90>}]
 dyn_jacketBack_03_r [{shapeType=Capsule, length=25, radius=11.5, shapeName=dyn_jacketBack_03_r_capsule, center=<2 0 10>, rotation=<rot 0 0 90>}]
 dyn_jacketBack_04_l [{shapeType=Capsule, length=28.5, radius=9.5, shapeName=dyn_jacketBack_04_l_capsule, center=<2 0 10>, rotation=<rot 0 0 90>}]
 dyn_jacketBack_04_r [{shapeType=Capsule, length=25.25, radius=9.5, shapeName=dyn_jacketBack_04_l_capsule, center=<2 0 10>, rotation=<rot 0 0 90>}]
 dyn_jacketBack_05_l [{shapeType=Capsule, length=31.5, radius=9.5, shapeName=dyn_jacketBack_05_l_capsule, center=<9.6914 0 10>, rotation=<rot 0 0 90>}]
 dyn_jacketBack_05_r [{shapeType=Capsule, length=23.25, radius=10.5, shapeName=dyn_jacketBack_05_r_capsule, center=<2 0 10>, rotation=<rot 0 0 90>}]
 dyn_jacketBack_mstr_l [{shapeType=Capsule, length=0.505, radius=0.505, rotation=<rot 0 -90 0>}]
 dyn_jacketBack_mstr_r [{shapeType=Capsule, length=0.505, radius=0.505, rotation=<rot 0 -90 0>}]
 dyn_jacketFront_01_l [{shapeType=Capsule, length=15, radius=8, shapeName=dyn_jacketFront_01_l_capsule, center=<5 0 4>, rotation=<rot 90 0 0>}]
 dyn_jacketFront_01_r [{shapeType=Capsule, length=11, radius=3, shapeName=dyn_jacketFront_01_r_capsule, center=<8.0011 0.0001 0.4904>, rotation=<rot 90 0 0>}]
 dyn_jacketFront_02_l [{shapeType=Capsule, length=26, radius=8, shapeName=dyn_jacketFront_02_l_capsule, center=<9 0 4>, rotation=<rot 0 0 90>}]
 dyn_jacketFront_02_r [{shapeType=Capsule, length=21.3305, radius=8, shapeName=dyn_jacketFront_02_r_capsule, center=<9.7618 0 4>, rotation=<rot 0 0 55.0222>}]
 dyn_jacketFront_03_l [
  {shapeType=Capsule, length=32.1117, radius=8, shapeName=dyn_jacketFront_03_l_capsule, center=<10 0 2>, rotation=<rot 0 0 -81.5266>}
  {
   shapeType=Capsule
   length=16
   radius=1.25
   shapeName=dyn_jacketFront_03_l_capsule
   center=<9.5898 0 0.5923>
   rotation=<rot 86.4659 179.9998 179.9998>
  }
 ]
 dyn_jacketFront_03_r [{
   shapeType=Capsule
   length=32.0248
   radius=8
   shapeName=dyn_jacketFront_03_l_capsule
   center=<9.7618 -1.6456 1.8216>
   rotation=<rot 0 0 47.9071>
  }]
 dyn_jacketFront_04_l [{shapeType=Capsule, length=32.349, radius=8, shapeName=dyn_jacketFront_04_l_capsule, center=<8.5574 0.0001 0>, rotation=<rot 0 0 -89.9999>}]
 dyn_jacketFront_04_r [{shapeType=Capsule, length=37.25, radius=8, shapeName=dyn_jacketFront_04_r_capsule, center=<5.3986 0 0>, rotation=<rot 0 0 45>}]
 dyn_jacketFront_05_l [{shapeType=Capsule, length=28.75, radius=8, shapeName=dyn_jacketFront_05_l_capsule, center=<7 0 0>, rotation=<rot 0 0 -89.9999>}]
 dyn_jacketFront_05_r [{
   shapeType=Capsule
   length=42.75
   radius=8
   shapeName=dyn_jacketFront_05_l_capsule
   center=<4.8209 0.4017 0.2777>
   rotation=<rot 5.569 8.1026 -124.2832>
  }]
 dyn_jacketFront_mstr_l [{shapeType=Capsule, length=0.505, radius=0.505, rotation=<rot 0 -90 0>}]
 dyn_jacketFront_mstr_r [{shapeType=Capsule, length=0.505, radius=8, center=<0 0 4>, rotation=<rot 0 -90 0>}]
 dyn_jacketSide_01_l [{shapeType=Capsule, length=14, radius=8, shapeName=dyn_jacketSide_01_l_capsule, center=<5 0 4>, rotation=<rot 0 0 75.4999>}]
 dyn_jacketSide_01_r [{shapeType=Capsule, length=14, radius=8, shapeName=dyn_jacketSide_01_l_capsule, center=<5 0 4>, rotation=<rot 0 0 75.4999>}]
 dyn_jacketSide_02_l [{shapeType=Capsule, length=18.25, radius=8, shapeName=dyn_jacketSide_02_l_capsule, center=<9 0 4>, rotation=<rot 0 0 71.9272>}]
 dyn_jacketSide_02_r [{shapeType=Capsule, length=37.2748, radius=10, shapeName=dyn_jacketSide_02_l_capsule, center=<9.7618 0 3>, rotation=<rot 0 0 89.9999>}]
 dyn_jacketSide_03_l [{shapeType=Capsule, length=31.5, radius=10, shapeName=dyn_jacketSide_03_l_capsule, center=<3.2695 0 2>, rotation=<rot 0 0 71.0282>}]
 dyn_jacketSide_03_r [{shapeType=Capsule, length=42.7748, radius=10, shapeName=dyn_jacketSide_03_l_capsule, center=<9.7618 0 2>, rotation=<rot 0 0 89.9999>}]
 dyn_jacketSide_04_l [{shapeType=Capsule, length=53, radius=10, shapeName=dyn_jacketSide_04_l_capsule, center=<6 0 2>, rotation=<rot 0 0 59.0735>}]
 dyn_jacketSide_04_r [{shapeType=Capsule, length=46, radius=10, shapeName=dyn_jacketSide_04_r_capsule, center=<7.2199 0 2>, rotation=<rot 0 0 90>}]
 dyn_jacketSide_05_l [{shapeType=Capsule, length=54.5, radius=10, shapeName=dyn_jacketSide_05_l_capsule, center=<8.4738 -3.2843 2>, rotation=<rot 0 0 56.9951>}]
 dyn_jacketSide_05_r [{shapeType=Capsule, length=46, radius=10, shapeName=dyn_jacketSide_05_r_capsule, center=<7.4186 0 2>, rotation=<rot 0 0 105.0003>}]
 dyn_jacketSide_mstr_l [{shapeType=Capsule, length=0.505, radius=0.505, rotation=<rot 0 -90 0>}]
 dyn_jacketSide_mstr_r [{shapeType=Capsule, length=0.505, radius=0.505, rotation=<rot 0 -90 0>}]
 pelvis [
  {shapeType=Capsule, length=12, radius=13, center=<-0.649 0.2801 -0.0002>, rotation=<rot 90 0 -3.3073>}
  {
   shapeType=Capsule
   length=15.3785
   radius=7.9107
   shapeName=pelvis_capsule
   center=<0.0122 -2.9989 -4.5558>
   rotation=<rot 80.001 -176.6924 -180>
  }
  {shapeType=Capsule, length=15.6212, radius=7.6681, shapeName=pelvis_capsule, center=<-0.082 -4.0081 5.3434>, rotation=<rot 77.8519 3.3074 0>}
 ]
 root [{shapeType=Sphere, radius=21.63, shapeName=root_sphere, center=<0.0002 -3.221 13.6507>}]
 thigh_l [{shapeType=Capsule, length=29.75, radius=9.747, center=<-21.4597 0.1549 0.7921>, rotation=<rot 90 12.9074 10.9407>}]
 thigh_r [{shapeType=Capsule, length=25.7052, radius=9.5281, center=<21.7868 0 0>, rotation=<rot 90 89.6476 89.6476>}]
}
constraints [
 {
  bone1Name=thigh_l
  bone2Name=pelvis
  swing1LimitDegrees=45
  swing1Motion=Free
  swing2LimitDegrees=45
  swing2Motion=Free
  twistLimitDegrees=45
  twistMotion=Free
 }
 {
  bone1Name=calf_l
  bone2Name=thigh_l
  swing1LimitDegrees=45
  swing1Motion=Free
  swing2LimitDegrees=45
  swing2Motion=Free
  twistLimitDegrees=45
  twistMotion=Free
 }
 {
  bone1Name=thigh_r
  bone2Name=pelvis
  swing1LimitDegrees=45
  swing1Motion=Free
  swing2LimitDegrees=45
  swing2Motion=Free
  twistLimitDegrees=45
  twistMotion=Free
 }
 {
  bone1Name=calf_r
  bone2Name=thigh_r
  swing1LimitDegrees=45
  swing1Motion=Free
  swing2LimitDegrees=45
  swing2Motion=Free
  twistLimitDegrees=45
  twistMotion=Free
 }
 {
  bone1Name=dyn_jacketSide_mstr_l
  bone2Name=pelvis
  swing1LimitDegrees=45
  swing1Motion=Free
  swing2LimitDegrees=45
  swing2Motion=Free
  twistLimitDegrees=45
  twistMotion=Free
 }
 {
  bone1Name=dyn_jacketBack_mstr_l
  bone2Name=pelvis
  swing1LimitDegrees=45
  swing1Motion=Free
  swing2LimitDegrees=45
  swing2Motion=Free
  twistLimitDegrees=45
  twistMotion=Free
 }
 {
  bone1Name=dyn_jacketBack_mstr_r
  bone2Name=pelvis
  swing1LimitDegrees=45
  swing1Motion=Free
  swing2LimitDegrees=45
  swing2Motion=Free
  twistLimitDegrees=45
  twistMotion=Free
 }
 {
  bone1Name=dyn_jacketSide_mstr_r
  bone2Name=pelvis
  swing1LimitDegrees=45
  swing1Motion=Free
  swing2LimitDegrees=45
  swing2Motion=Free
  twistLimitDegrees=45
  twistMotion=Free
 }
 {
  bone1Name=dyn_jacketFront_mstr_l
  bone2Name=pelvis
  swing1LimitDegrees=45
  swing1Motion=Free
  swing2LimitDegrees=45
  swing2Motion=Free
  twistLimitDegrees=45
  twistMotion=Free
 }
 {
  bone1Name=dyn_jacketFront_mstr_r
  bone2Name=pelvis
  swing1LimitDegrees=45
  swing1Motion=Free
  swing2LimitDegrees=45
  swing2Motion=Free
  twistLimitDegrees=45
  twistMotion=Free
 }
 {bone1Name=dyn_jacketFront_01_r, bone2Name=dyn_jacketFront_mstr_r, swing1LimitDegrees=45, swing2LimitDegrees=45, twistLimitDegrees=5}
 {bone1Name=dyn_jacketFront_02_r, bone2Name=dyn_jacketFront_01_r, swing1LimitDegrees=45, swing2LimitDegrees=45, twistLimitDegrees=5}
 {bone1Name=dyn_jacketFront_03_r, bone2Name=dyn_jacketFront_02_r, swing1LimitDegrees=45, swing2LimitDegrees=45, twistLimitDegrees=5}
 {bone1Name=dyn_jacketFront_04_r, bone2Name=dyn_jacketFront_03_r, swing1LimitDegrees=45, swing2LimitDegrees=45, twistLimitDegrees=5}
 {bone1Name=dyn_jacketFront_01_l, bone2Name=dyn_jacketFront_mstr_l, swing1LimitDegrees=20, swing2LimitDegrees=45, twistLimitDegrees=5}
 {bone1Name=dyn_jacketFront_02_l, bone2Name=dyn_jacketFront_01_l, swing1LimitDegrees=45, swing2LimitDegrees=45, twistLimitDegrees=5}
 {bone1Name=dyn_jacketFront_03_l, bone2Name=dyn_jacketFront_02_l, swing1LimitDegrees=45, swing2LimitDegrees=45, twistLimitDegrees=5}
 {bone1Name=dyn_jacketFront_04_l, bone2Name=dyn_jacketFront_03_l, swing1LimitDegrees=45, swing2LimitDegrees=45, twistLimitDegrees=5}
 {bone1Name=dyn_jacketSide_01_r, bone2Name=dyn_jacketSide_mstr_r, swing1LimitDegrees=45, swing2LimitDegrees=45, twistLimitDegrees=5}
 {bone1Name=dyn_jacketSide_02_r, bone2Name=dyn_jacketSide_01_r, swing1LimitDegrees=45, swing2LimitDegrees=45, twistLimitDegrees=5}
 {bone1Name=dyn_jacketSide_03_r, bone2Name=dyn_jacketSide_02_r, swing1LimitDegrees=45, swing2LimitDegrees=45, twistLimitDegrees=5}
 {bone1Name=dyn_jacketSide_04_r, bone2Name=dyn_jacketSide_03_r, swing1LimitDegrees=45, swing2LimitDegrees=45, twistLimitDegrees=5}
 {bone1Name=dyn_jacketBack_01_r, bone2Name=dyn_jacketBack_mstr_r, swing1LimitDegrees=45, swing2LimitDegrees=45, twistLimitDegrees=5}
 {bone1Name=dyn_jacketBack_02_r, bone2Name=dyn_jacketBack_01_r, swing1LimitDegrees=45, swing2LimitDegrees=45, twistLimitDegrees=5}
 {bone1Name=dyn_jacketBack_03_r, bone2Name=dyn_jacketBack_02_r, swing1LimitDegrees=45, swing2LimitDegrees=45, twistLimitDegrees=5}
 {bone1Name=dyn_jacketBack_04_r, bone2Name=dyn_jacketBack_03_r, swing1LimitDegrees=45, swing2LimitDegrees=45, twistLimitDegrees=5}
 {bone1Name=dyn_jacketBack_01_l, bone2Name=dyn_jacketBack_mstr_l, swing1LimitDegrees=20, swing2LimitDegrees=45, twistLimitDegrees=5}
 {bone1Name=dyn_jacketBack_02_l, bone2Name=dyn_jacketBack_01_l, swing1LimitDegrees=45, swing2LimitDegrees=45, twistLimitDegrees=5}
 {bone1Name=dyn_jacketBack_03_l, bone2Name=dyn_jacketBack_02_l, swing1LimitDegrees=45, swing2LimitDegrees=45, twistLimitDegrees=5}
 {bone1Name=dyn_jacketBack_04_l, bone2Name=dyn_jacketBack_03_l, swing1LimitDegrees=45, swing2LimitDegrees=45, twistLimitDegrees=5}
 {bone1Name=dyn_jacketSide_01_l, bone2Name=dyn_jacketSide_mstr_l, swing1LimitDegrees=12, swing2LimitDegrees=40, twistLimitDegrees=5}
 {bone1Name=dyn_jacketSide_02_l, bone2Name=dyn_jacketSide_01_l, swing1LimitDegrees=45, swing2LimitDegrees=45, twistLimitDegrees=5}
 {bone1Name=dyn_jacketSide_03_l, bone2Name=dyn_jacketSide_02_l, swing1LimitDegrees=45, swing2LimitDegrees=45, twistLimitDegrees=5}
 {bone1Name=dyn_jacketSide_04_l, bone2Name=dyn_jacketSide_03_l, swing1LimitDegrees=45, swing2LimitDegrees=45, twistLimitDegrees=5}
 {
  bone1Name=pelvis
  bone2Name=root
  swing1LimitDegrees=45
  swing1Motion=Free
  swing2LimitDegrees=45
  swing2Motion=Free
  twistLimitDegrees=45
  twistMotion=Free
 }
 {bone1Name=dyn_jacketFront_05_r, bone2Name=dyn_jacketFront_04_r, swing1LimitDegrees=45, swing2LimitDegrees=45, twistLimitDegrees=5}
 {bone1Name=dyn_jacketFront_05_l, bone2Name=dyn_jacketFront_04_l, swing1LimitDegrees=45, swing2LimitDegrees=45, twistLimitDegrees=5}
 {bone1Name=dyn_jacketSide_05_r, bone2Name=dyn_jacketSide_04_r, swing1LimitDegrees=45, swing2LimitDegrees=45, twistLimitDegrees=5}
 {bone1Name=dyn_jacketSide_05_l, bone2Name=dyn_jacketSide_04_l, swing1LimitDegrees=45, swing2LimitDegrees=45, twistLimitDegrees=5}
 {bone1Name=dyn_jacketBack_05_l, bone2Name=dyn_jacketBack_04_l, swing1LimitDegrees=45, swing2LimitDegrees=45, twistLimitDegrees=5}
 {bone1Name=dyn_jacketBack_05_r, bone2Name=dyn_jacketBack_04_r, swing1LimitDegrees=45, swing2LimitDegrees=45, twistLimitDegrees=5}
 {
  bone1Name=dyn_holster
  bone2Name=thigh_r
  swing1LimitDegrees=45
  swing1Motion=Free
  swing2LimitDegrees=45
  swing2Motion=Free
  twistLimitDegrees=45
  twistMotion=Free
 }
 {
  bone1Name=dyn_jacketSide_02_r
  bone2Name=dyn_jacketFront_02_r
  swing1LimitDegrees=45
  swing1Motion=Free
  swing2LimitDegrees=45
  swing2Motion=Free
  twistLimitDegrees=45
  twistMotion=Free
 }
 {
  bone1Name=dyn_jacketSide_01_l
  bone2Name=dyn_jacketFront_01_l
  swing1LimitDegrees=45
  swing1Motion=Free
  swing2LimitDegrees=45
  swing2Motion=Free
  twistLimitDegrees=45
  twistMotion=Free
 }
 {
  bone1Name=dyn_jacketSide_02_l
  bone2Name=dyn_jacketFront_02_l
  swing1LimitDegrees=45
  swing1Motion=Free
  swing2LimitDegrees=45
  swing2Motion=Free
  twistLimitDegrees=45
  twistMotion=Free
 }
 {
  bone1Name=dyn_jacketSide_03_l
  bone2Name=dyn_jacketFront_03_l
  swing1LimitDegrees=45
  swing1Motion=Free
  swing2LimitDegrees=45
  swing2Motion=Free
  twistLimitDegrees=45
  twistMotion=Free
 }
 {
  bone1Name=dyn_jacketSide_01_r
  bone2Name=dyn_jacketFront_01_r
  swing1LimitDegrees=45
  swing1Motion=Free
  swing2LimitDegrees=45
  swing2Motion=Free
  twistLimitDegrees=45
  twistMotion=Free
 }
 {
  bone1Name=dyn_jacketSide_03_r
  bone2Name=dyn_jacketFront_03_r
  swing1LimitDegrees=45
  swing1Motion=Free
  swing2LimitDegrees=45
  swing2Motion=Free
  twistLimitDegrees=45
  twistMotion=Free
 }
 {
  bone1Name=dyn_ground_root
  bone2Name=root
  swing1LimitDegrees=45
  swing1Motion=Free
  swing2LimitDegrees=45
  swing2Motion=Free
  twistLimitDegrees=45
  twistMotion=Free
 }
]
