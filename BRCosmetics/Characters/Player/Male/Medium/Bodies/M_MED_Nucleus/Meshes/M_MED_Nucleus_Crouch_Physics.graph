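# Physics asset for the crouch variant of the male medium Nucleus body, covering the lower body, a holster and the jacket's dynamic bones.
# Six five-segment capsule chains (front, side and back, left and right) hang from pelvis-linked master bones, mostly with tight twist limits.
# Pelvis and leg capsules plus two root spheres, one very large and offset like a floor, provide body shapes.
# Body joints, master links, holster-to-thigh and one right front-to-side cross-link rotate freely.
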PhysicsAsset /BRCosmetics/Characters/Player/Male/Medium/Bodies/M_MED_Nucleus/Meshes/M_MED_Nucleus_Crouch_Physics
bodies {
 calf_l [
  {shapeType=Capsule, length=48.3142, radius=8.75, center=<-23.7288 -2.5808 0.0141>, rotation=<rot 90 90 84>}
  {
   shapeType=Capsule
   length=38.6216
   radius=4.9618
   shapeName=calf_l_capsule
   center=<-14.97 5.6017 -0.0002>
   rotation=<rot 90 -18.9704 -24.9704>
  }
 ]
 calf_r [{shapeType=Capsule, length=59.2071, radius=11.5191, center=<25.0282 2.68 1.7816>, rotation=<rot 90 12.9074 11.169>}]
 dyn_holster [{
   shapeType=Capsule
   length=15.9376
   radius=10.1123
   shapeName=dyn_holster_capsule
   center=<3.6367 -1.9359 1.0029>
   rotation=<rot -0.4561 -94.8553 -66.2518>
  }]
 dyn_jacketBack_01_l [{shapeType=Capsule, length=7.8918, radius=2.1082, shapeName=dyn_jacketBack_01_l_capsule, rotation=<rot 0 0 90>}]
 dyn_jacketBack_01_r [{shapeType=Capsule, length=7.7086, radius=2.2914, shapeName=dyn_jacketBack_01_r_capsule, center=<5.2957 0 0>, rotation=<rot 0 0 89.9999>}]
 dyn_jacketBack_02_l [{shapeType=Capsule, length=12, radius=5.5, shapeName=dyn_jacketBack_02_l_capsule, center=<2 0 0>, rotation=<rot 90 0 0>}]
 dyn_jacketBack_02_r [{shapeType=Capsule, length=12, radius=5.5, shapeName=dyn_jacketBack_02_r_capsule, center=<10.2621 0 -0.0001>, rotation=<rot 90 0 0>}]
 dyn_jacketBack_03_l [{shapeType=Capsule, length=28, radius=5, shapeName=dyn_jacketBack_03_l_capsule, center=<7.7254 0 0.0003>, rotation=<rot 0 0 89.9999>}]
 dyn_jacketBack_03_r [{shapeType=Capsule, length=28, radius=5, shapeName=dyn_jacketBack_03_r_capsule, center=<8.0908 0 0>, rotation=<rot 0 0 89.9999>}]
 dyn_jacketBack_04_l [{shapeType=Capsule, length=17.75, radius=6.5, shapeName=dyn_jacketBack_04_l_capsule, center=<7.7254 0 0.0003>, rotation=<rot 0 0 89.9999>}]
 dyn_jacketBack_04_r [{shapeType=Capsule, length=20, radius=4.25, shapeName=dyn_jacketBack_04_r_capsule, center=<8.0908 0 0>, rotation=<rot 0 0 89.9999>}]
 dyn_jacketBack_05_l [{shapeType=Capsule, length=24.5, radius=5, shapeName=dyn_jacketBack_05_l_capsule, center=<7.7254 0 0.0003>, rotation=<rot 0 0 89.9999>}]
 dyn_jacketBack_05_r [{shapeType=Capsule, length=21.25, radius=7, shapeName=dyn_jacketBack_05_r_capsule, center=<8.0908 0 0>, rotation=<rot 0 0 89.9999>}]
 dyn_jacketBack_mstr_l [{shapeType=Capsule, length=0.505, radius=0.505, rotation=<rot 0 -90 0>}]
 dyn_jacketBack_mstr_r [{shapeType=Capsule, length=0.505, radius=0.505, rotation=<rot 0 -90 0>}]
 dyn_jacketFront_01_l [{
   shapeType=Capsule
   length=18.5
   radius=3.25
   shapeName=dyn_jacketFront_01_l_capsule
   center=<5 0.0001 1.7355>
   rotation=<rot 82.7218 -179.9999 -179.9999>
  }]
 dyn_jacketFront_01_r [{shapeType=Capsule, length=15, radius=3, shapeName=dyn_jacketFront_01_r_capsule, center=<5 0 0>, rotation=<rot 90 0 0>}]
 dyn_jacketFront_02_l [{shapeType=Capsule, length=18.75, radius=3, shapeName=dyn_jacketFront_02_l_capsule, center=<11.4349 0.0001 2.7229>, rotation=<rot 90 0 0>}]
 dyn_jacketFront_02_r [{shapeType=Capsule, length=15, radius=3, shapeName=dyn_jacketFront_02_r_capsule, center=<9 0 0>, rotation=<rot 90 0 0>}]
 dyn_jacketFront_03_l [{shapeType=Capsule, length=10.75, radius=3.5, shapeName=dyn_jacketFront_03_l_capsule, center=<8 0 0>, rotation=<rot 90 0 0>}]
 dyn_jacketFront_03_r [{shapeType=Capsule, length=15, radius=3, shapeName=dyn_jacketFront_03_r_capsule, center=<8 0 0>, rotation=<rot 90 0 0>}]
 dyn_jacketFront_04_l [{
   shapeType=Capsule
   length=14.25
   radius=3.5
   shapeName=dyn_jacketFront_04_l_capsule
   center=<3.417 -0.0001 0>
   rotation=<rot 85.1666 -179.9999 0.0001>
  }]
 dyn_jacketFront_04_r [{shapeType=Capsule, length=15, radius=3, shapeName=dyn_jacketFront_04_r_capsule, center=<6 0 0>, rotation=<rot 90 0 0>}]
 dyn_jacketFront_05_l [{shapeType=Capsule, length=9.25, radius=2.25, shapeName=dyn_jacketFront_05_l_capsule, center=<3.6992 0.0003 -0.6076>, rotation=<rot 90 0 0>}]
 dyn_jacketFront_05_r [{shapeType=Capsule, length=6.25, radius=5.25, shapeName=dyn_jacketFront_05_r_capsule, center=<2.9134 -0.0001 -3.1165>, rotation=<rot 90 0 0>}]
 dyn_jacketFront_mstr_l [{shapeType=Capsule, length=0.505, radius=0.505, rotation=<rot 0 -90 0>}]
 dyn_jacketFront_mstr_r [{shapeType=Capsule, length=0.505, radius=0.505, rotation=<rot 0 -90 0>}]
 dyn_jacketSide_01_l [{shapeType=Capsule, length=12, radius=6, shapeName=dyn_jacketSide_01_l_capsule, center=<6 0 0>, rotation=<rot 105 90 0>}]
 dyn_jacketSide_01_r [{shapeType=Capsule, length=15, radius=3, shapeName=dyn_jacketSide_01_r_capsule, center=<6 0 0>, rotation=<rot 0 0 89>}]
 dyn_jacketSide_02_l [{
   shapeType=Capsule
   length=22
   radius=2.75
   shapeName=dyn_jacketSide_02_l_capsule
   center=<8.6948 -0.8904 -0.2384>
   rotation=<rot 74.9995 -89.9995 180>
  }]
 dyn_jacketSide_02_r [{shapeType=Capsule, length=13.25, radius=3.75, shapeName=dyn_jacketSide_02_r_capsule, center=<6 0 0>, rotation=<rot 0 0 89>}]
 dyn_jacketSide_03_l [{
   shapeType=Capsule
   length=20.5
   radius=2.75
   shapeName=dyn_jacketSide_03_l_capsule
   center=<5.9853 -0.0003 -0.0002>
   rotation=<rot 74.9995 -89.9995 180>
  }]
 dyn_jacketSide_03_r [{shapeType=Capsule, length=25, radius=5, shapeName=dyn_jacketSide_03_r_capsule, center=<6 0 0>, rotation=<rot 0 0 89>}]
 dyn_jacketSide_04_l [{shapeType=Capsule, length=23.5, radius=6.5, shapeName=dyn_jacketSide_04_l_capsule, center=<8 0 0>, rotation=<rot 0 0 65.4362>}]
 dyn_jacketSide_04_r [{shapeType=Capsule, length=25, radius=5, shapeName=dyn_jacketSide_04_r_capsule, center=<6 0 0>, rotation=<rot 0 0 89>}]
 dyn_jacketSide_05_l [{shapeType=Capsule, length=29.5, radius=6.25, shapeName=dyn_jacketSide_05_l_capsule, center=<3.477 0 0>, rotation=<rot 0 0 45>}]
 dyn_jacketSide_05_r [{
   shapeType=Capsule
   length=18
   radius=3
   shapeName=dyn_jacketSide_05_r_capsule
   center=<3.5569 0.4172 -0.0049>
   rotation=<rot -0.1736 15.0029 101.1133>
  }]
 dyn_jacketSide_mstr_l [{shapeType=Capsule, length=0.505, radius=0.505, rotation=<rot 0 -90 0>}]
 dyn_jacketSide_mstr_r [{shapeType=Capsule, length=0.505, radius=0.505, rotation=<rot 0 -90 0>}]
 pelvis [
  {shapeType=Capsule, length=12.7328, radius=12, center=<-0.649 0.2801 -0.0002>, rotation=<rot 90 0 -3.3073>}
  {
   shapeType=Capsule
   length=15.3785
   radius=7.9107
   shapeName=pelvis_capsule
   center=<-0.0808 -1.3903 -4.5558>
   rotation=<rot 80.001 -176.6924 -180>
  }
  {shapeType=Capsule, length=15.6212, radius=7.6681, shapeName=pelvis_capsule, center=<-0.2328 -1.3991 5.3434>, rotation=<rot 77.8519 3.3074 0>}
 ]
 root [
  {shapeType=Sphere, radius=232.9375, shapeName=root_sphere, center=<7.1775 10.2377 -217.4697>}
  {shapeType=Sphere, radius=32.88, shapeName=root_sphere, center=<0.0002 2.4787 8.5972>}
 ]
 thigh_l [
  {shapeType=Capsule, length=30.9863, radius=9.497, center=<-21.4904 1.0476 -0.2548>, rotation=<rot 90 12.9074 10.9407>}
  {shapeType=Capsule, length=25.25, radius=8.5, shapeName=thigh_l_capsule, center=<-39.0775 11.7854 2.858>, rotation=<rot 0 0 76.3717>}
  {
   shapeType=Capsule
   length=30.75
   radius=5.25
   shapeName=thigh_l_capsule
   center=<-13.8252 12.2198 1.0694>
   rotation=<rot 0.7718 -4.3606 80.6286>
  }
 ]
 thigh_r [{shapeType=Capsule, length=24.4379, radius=11.5454, center=<21.7868 0 0>, rotation=<rot 90 0 0>}]
}
constraints [
 {
  bone1Name=thigh_l
  bone2Name=pelvis
  swing1LimitDegrees=45
  swing1Motion=Free
  swing2LimitDegrees=45
  swing2Motion=Free
  twistLimitDegrees=45
  twistMotion=Free
 }
 {
  bone1Name=calf_l
  bone2Name=thigh_l
  swing1LimitDegrees=45
  swing1Motion=Free
  swing2LimitDegrees=45
  swing2Motion=Free
  twistLimitDegrees=45
  twistMotion=Free
 }
 {
  bone1Name=thigh_r
  bone2Name=pelvis
  swing1LimitDegrees=45
  swing1Motion=Free
  swing2LimitDegrees=45
  swing2Motion=Free
  twistLimitDegrees=45
  twistMotion=Free
 }
 {
  bone1Name=calf_r
  bone2Name=thigh_r
  swing1LimitDegrees=45
  swing1Motion=Free
  swing2LimitDegrees=45
  swing2Motion=Free
  twistLimitDegrees=45
  twistMotion=Free
 }
 {
  bone1Name=dyn_jacketSide_mstr_l
  bone2Name=pelvis
  swing1LimitDegrees=45
  swing1Motion=Free
  swing2LimitDegrees=45
  swing2Motion=Free
  twistLimitDegrees=45
  twistMotion=Free
 }
 {
  bone1Name=dyn_jacketBack_mstr_l
  bone2Name=pelvis
  swing1LimitDegrees=45
  swing1Motion=Free
  swing2LimitDegrees=45
  swing2Motion=Free
  twistLimitDegrees=45
  twistMotion=Free
 }
 {
  bone1Name=dyn_jacketBack_mstr_r
  bone2Name=pelvis
  swing1LimitDegrees=45
  swing1Motion=Free
  swing2LimitDegrees=45
  swing2Motion=Free
  twistLimitDegrees=45
  twistMotion=Free
 }
 {
  bone1Name=dyn_jacketSide_mstr_r
  bone2Name=pelvis
  swing1LimitDegrees=45
  swing1Motion=Free
  swing2LimitDegrees=45
  swing2Motion=Free
  twistLimitDegrees=45
  twistMotion=Free
 }
 {
  bone1Name=dyn_jacketFront_mstr_l
  bone2Name=pelvis
  swing1LimitDegrees=45
  swing1Motion=Free
  swing2LimitDegrees=45
  swing2Motion=Free
  twistLimitDegrees=45
  twistMotion=Free
 }
 {
  bone1Name=dyn_jacketFront_mstr_r
  bone2Name=pelvis
  swing1LimitDegrees=45
  swing1Motion=Free
  swing2LimitDegrees=45
  swing2Motion=Free
  twistLimitDegrees=45
  twistMotion=Free
 }
 {bone1Name=dyn_jacketFront_01_r, bone2Name=dyn_jacketFront_mstr_r, swing1LimitDegrees=45, swing2LimitDegrees=45, twistLimitDegrees=0.5}
 {bone1Name=dyn_jacketFront_02_r, bone2Name=dyn_jacketFront_01_r, swing1LimitDegrees=45, swing2LimitDegrees=45, twistLimitDegrees=0.5}
 {bone1Name=dyn_jacketFront_03_r, bone2Name=dyn_jacketFront_02_r, swing1LimitDegrees=45, swing2LimitDegrees=45, twistLimitDegrees=0.5}
 {bone1Name=dyn_jacketFront_04_r, bone2Name=dyn_jacketFront_03_r, swing1LimitDegrees=45, swing2LimitDegrees=45, twistLimitDegrees=0.5}
 {bone1Name=dyn_jacketFront_01_l, bone2Name=dyn_jacketFront_mstr_l, swing1LimitDegrees=45, swing2LimitDegrees=45, twistLimitDegrees=0.2}
 {bone1Name=dyn_jacketFront_02_l, bone2Name=dyn_jacketFront_01_l, swing1LimitDegrees=45, swing2LimitDegrees=45, twistLimitDegrees=25}
 {bone1Name=dyn_jacketFront_03_l, bone2Name=dyn_jacketFront_02_l, swing1LimitDegrees=45, swing2LimitDegrees=45, twistLimitDegrees=25}
 {bone1Name=dyn_jacketFront_04_l, bone2Name=dyn_jacketFront_03_l, swing1LimitDegrees=25, swing2LimitDegrees=45, twistLimitDegrees=30}
 {bone1Name=dyn_jacketSide_01_r, bone2Name=dyn_jacketSide_mstr_r, swing1LimitDegrees=45, swing2LimitDegrees=45, twistLimitDegrees=0.5}
 {bone1Name=dyn_jacketSide_02_r, bone2Name=dyn_jacketSide_01_r, swing1LimitDegrees=45, swing2LimitDegrees=45, twistLimitDegrees=0.5}
 {bone1Name=dyn_jacketSide_03_r, bone2Name=dyn_jacketSide_02_r, swing1LimitDegrees=45, swing2LimitDegrees=45, twistLimitDegrees=0.5}
 {bone1Name=dyn_jacketSide_04_r, bone2Name=dyn_jacketSide_03_r, swing1LimitDegrees=45, swing2LimitDegrees=45, twistLimitDegrees=0.5}
 {bone1Name=dyn_jacketBack_01_r, bone2Name=dyn_jacketBack_mstr_r, swing1LimitDegrees=45, swing2LimitDegrees=45, twistLimitDegrees=0.5}
 {bone1Name=dyn_jacketBack_02_r, bone2Name=dyn_jacketBack_01_r, swing1LimitDegrees=45, swing2LimitDegrees=45, twistLimitDegrees=0.5}
 {bone1Name=dyn_jacketBack_03_r, bone2Name=dyn_jacketBack_02_r, swing1LimitDegrees=45, swing2LimitDegrees=45, twistLimitDegrees=0.5}
 {bone1Name=dyn_jacketBack_04_r, bone2Name=dyn_jacketBack_03_r, swing1LimitDegrees=45, swing2LimitDegrees=45, twistLimitDegrees=0.5}
 {bone1Name=dyn_jacketBack_01_l, bone2Name=dyn_jacketBack_mstr_l, swing1LimitDegrees=45, swing2LimitDegrees=45, twistLimitDegrees=0.5}
 {bone1Name=dyn_jacketBack_02_l, bone2Name=dyn_jacketBack_01_l, swing1LimitDegrees=45, swing2LimitDegrees=45, twistLimitDegrees=0.5}
 {bone1Name=dyn_jacketBack_03_l, bone2Name=dyn_jacketBack_02_l, swing1LimitDegrees=45, swing2LimitDegrees=45, twistLimitDegrees=0.5}
 {bone1Name=dyn_jacketBack_04_l, bone2Name=dyn_jacketBack_03_l, swing1LimitDegrees=45, swing2LimitDegrees=45, twistLimitDegrees=0.5}
 {bone1Name=dyn_jacketSide_01_l, bone2Name=dyn_jacketSide_mstr_l, swing1LimitDegrees=45, swing2LimitDegrees=45, twistLimitDegrees=0.5}
 {bone1Name=dyn_jacketSide_02_l, bone2Name=dyn_jacketSide_01_l, swing1LimitDegrees=45, swing2LimitDegrees=45, twistLimitDegrees=0.5}
 {bone1Name=dyn_jacketSide_03_l, bone2Name=dyn_jacketSide_02_l, swing1LimitDegrees=45, swing2LimitDegrees=45, twistLimitDegrees=0.5}
 {bone1Name=dyn_jacketSide_04_l, bone2Name=dyn_jacketSide_03_l, swing1LimitDegrees=45, swing2LimitDegrees=45, twistLimitDegrees=0.5}
 {
  bone1Name=pelvis
  bone2Name=root
  swing1LimitDegrees=45
  swing1Motion=Free
  swing2LimitDegrees=45
  swing2Motion=Free
  twistLimitDegrees=45
  twistMotion=Free
 }
 {
  bone1Name=dyn_jacketSide_01_r
  bone2Name=dyn_jacketFront_01_r
  swing1LimitDegrees=45
  swing1Motion=Free
  swing2LimitDegrees=45
  swing2Motion=Free
  twistLimitDegrees=45
  twistMotion=Free
 }
 {bone1Name=dyn_jacketFront_05_r, bone2Name=dyn_jacketFront_04_r, swing1LimitDegrees=45, swing2LimitDegrees=45, twistLimitDegrees=0.5}
 {bone1Name=dyn_jacketFront_05_l, bone2Name=dyn_jacketFront_04_l, swing1LimitDegrees=25, swing2LimitDegrees=45, twistLimitDegrees=30}
 {bone1Name=dyn_jacketSide_05_r, bone2Name=dyn_jacketSide_04_r, swing1LimitDegrees=45, swing2LimitDegrees=45, twistLimitDegrees=0.5}
 {bone1Name=dyn_jacketSide_05_l, bone2Name=dyn_jacketSide_04_l, swing1LimitDegrees=45, swing2LimitDegrees=45, twistLimitDegrees=2}
 {
  bone1Name=dyn_holster
  bone2Name=thigh_r
  swing1LimitDegrees=45
  swing1Motion=Free
  swing2LimitDegrees=45
  swing2Motion=Free
  twistLimitDegrees=45
  twistMotion=Free
 }
 {bone1Name=dyn_jacketBack_05_r, bone2Name=dyn_jacketBack_04_r, swing1LimitDegrees=45, swing2LimitDegrees=45, twistLimitDegrees=2}
 {bone1Name=dyn_jacketBack_05_l, bone2Name=dyn_jacketBack_04_l, swing1LimitDegrees=45, swing2LimitDegrees=45, twistLimitDegrees=2}
]
