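# Post-process style AnimBP for the M_MED_Nucleus cosmetic body, animating its coat, collar, zipper and holster over the pose copied from the parent mesh.
# Control Rigs and bone corrections feed cached poses into RigidBody branches for crouching, skydiving, front end and normal play.
# AnimDynamics and constraints finish zipper, collar and coat tails.
# The event graph derives limb-rotation alphas, state flags, gravity override and simulation-space rotation.
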AnimBlueprint /BRCosmetics/Characters/Player/Male/Medium/Bodies/M_MED_Nucleus/Meshes/M_MED_Nucleus_AnimBP
target_skeleton=&/M_MED_Nucleus_Skeleton
graphs [
 AnimGraph {nodes [
   n1 = Look At - Bone: dyn_jacketBack_mstr_l(Alpha=1) {
    alphaBoolBlend {blendOption=Linear}
    alphaScaleBiasClamp {bClampResult=true, clampMax=0.3}
    boneToModify {boneName=dyn_jacketBack_mstr_l}
    lookAtLocation=<-97.5218 -32.2726 -68.1181>
    lookAtTarget {boneReference {boneName=pelvis}}
    lookAt_Axis {axis=<1 0>}
   }
   n2 = Look At - Bone: dyn_jacketBack_mstr_r(ComponentPose=@n1, Alpha=1) {
    alphaBoolBlend {blendOption=Linear}
    alphaScaleBiasClamp {bClampResult=true, clampMax=0.5}
    boneToModify {boneName=dyn_jacketBack_mstr_r}
    lookAtLocation=<-90.5576 -82.3551 -16.2656>
    lookAtTarget {boneReference {boneName=pelvis}}
    lookAt_Axis {axis=<1 0>}
   }
   Look At - Bone: dyn_jacketSide_mstr_l(ComponentPose=@n2, Alpha=1) {
    alphaBoolBlend {blendOption=Linear}
    alphaScaleBiasClamp {bClampResult=true, clampMax=0.3}
    boneToModify {boneName=dyn_jacketSide_mstr_l}
    lookAtLocation=<-102.3621 -56.129 -61.7368>
    lookAtTarget {boneReference {boneName=pelvis}}
    lookAt_Axis {axis=<1 0>}
   }
   n3 = "Constraint - dyn_jacketFront_mstr_l (P)"(ComponentPose=Local To Component(LocalPose=Use cached pose 'AnimCache'()), Alpha=1, ConstraintWeights_0=1) {
    alphaBoolBlend {blendOption=Linear}
    boneToModify {boneName=dyn_jacketFront_mstr_l}
    constraintSetup [{targetBone {boneName=thigh_l}}]
    constraintWeights [1]
   }
   n4 = "Constraint - dyn_jacketFront_mstr_r (P)"(ComponentPose=@n3, Alpha=1, ConstraintWeights_0=1) {
    alphaBoolBlend {blendOption=Linear}
    boneToModify {boneName=dyn_jacketFront_mstr_r}
    constraintSetup [{targetBone {boneName=thigh_r}}]
    constraintWeights [1]
   }
   n5 = "Transform (Modify) Bone - Bone: dyn_jacketFront_mstr_l"(ComponentPose=@n4, Alpha=1, Translation="0.000,-0.000,2", Rotation="25,0.000,0.000", Scale="1,1,1") {
    rotationMode=BMM_Additive
    rotationSpace=BCS_BoneSpace
    scaleMode=BMM_Additive
    scaleSpace=BCS_BoneSpace
    translationMode=BMM_Additive
    translationSpace=BCS_BoneSpace
    alphaBoolBlend {blendOption=Linear}
    boneToModify {boneName=dyn_jacketFront_mstr_l}
    rotation {pitch=25}
    translation {z=2}
   }
   n6 = "Transform (Modify) Bone - Bone: dyn_jacketFront_mstr_r"(ComponentPose=@n5, Alpha=1, Translation="0.000,0,2", Rotation="25,0.000,0.000", Scale="1,1,1") {
    rotationMode=BMM_Additive
    rotationSpace=BCS_BoneSpace
    scaleMode=BMM_Additive
    scaleSpace=BCS_BoneSpace
    translationMode=BMM_Additive
    translationSpace=BCS_BoneSpace
    alphaBoolBlend {blendOption=Linear}
    boneToModify {boneName=dyn_jacketFront_mstr_r}
    rotation {pitch=25}
    translation {z=2}
   }
   n7 = "Transform (Modify) Bone - Bone: dyn_jacketSide_mstr_l"(ComponentPose=@n6, Alpha=1, Translation="0.000,-0.000,3", Rotation="30,0.000,0", Scale="1,1,1") {
    rotationMode=BMM_Additive
    rotationSpace=BCS_BoneSpace
    scaleMode=BMM_Additive
    scaleSpace=BCS_BoneSpace
    translationMode=BMM_Additive
    translationSpace=BCS_BoneSpace
    alphaBoolBlend {blendOption=Linear}
    boneToModify {boneName=dyn_jacketSide_mstr_l}
    rotation {pitch=30}
    translation {z=3}
   }
   n8 = "Transform (Modify) Bone - Bone: dyn_jacketSide_mstr_r"(ComponentPose=@n7, Alpha=1, Translation="0.000,-0.000,3", Rotation="30,0.000,0.000", Scale="1,1,1") {
    rotationMode=BMM_Additive
    rotationSpace=BCS_BoneSpace
    scaleMode=BMM_Additive
    scaleSpace=BCS_BoneSpace
    translationMode=BMM_Additive
    translationSpace=BCS_BoneSpace
    alphaBoolBlend {blendOption=Linear}
    boneToModify {boneName=dyn_jacketSide_mstr_r}
    rotation {pitch=30}
    translation {z=3}
   }
   n9 = "Transform (Modify) Bone - Bone: dyn_jacketBack_mstr_l"(ComponentPose=@n8, Alpha=1, Translation="0.000,-0.000,3", Rotation="5,0.000,0.000", Scale="1,1,1") {
    rotationMode=BMM_Additive
    rotationSpace=BCS_BoneSpace
    scaleMode=BMM_Additive
    scaleSpace=BCS_BoneSpace
    translationMode=BMM_Additive
    translationSpace=BCS_BoneSpace
    alphaBoolBlend {blendOption=Linear}
    boneToModify {boneName=dyn_jacketBack_mstr_l}
    rotation {pitch=5}
    translation {z=3}
   }
   n10 = "Transform (Modify) Bone - Bone: dyn_jacketBack_mstr_r"(ComponentPose=@n9, Alpha=1, Translation="-0.000,-0.000,3", Rotation="5,0,0.000", Scale="1,1,1") {
    rotationMode=BMM_Additive
    rotationSpace=BCS_BoneSpace
    scaleMode=BMM_Additive
    scaleSpace=BCS_BoneSpace
    translationMode=BMM_Additive
    translationSpace=BCS_BoneSpace
    alphaBoolBlend {blendOption=Linear}
    boneToModify {boneName=dyn_jacketBack_mstr_r}
    rotation {pitch=5}
    translation {z=3}
   }
   n11 = Get GravityOverride()
   n12 = RigidBody!(ComponentPose=@n10, Alpha=Get Is not Front End() -> (Is not Front End:double), OverrideWorldGravity:Vector=@n11, ExternalForce) -> (Pose) {
    bClampLinearTranslationLimitToRefPose=true
    bOverrideWorldGravity=true
    overridePhysicsAsset=&/M_MED_Nucleus_Coat_Physics
    simulationSpace=BaseBoneSpace
    alphaBoolBlend {blendOption=Linear}
    baseBoneRef {boneName=root}
    componentAppliedLinearAccClamp=<500 500 500>
    overrideWorldGravity {z=150}
   }
   n13 = Get Is in Front End() -> (Is in Front End:double)
   n14 = RigidBody(ComponentPose=@n12, Alpha=@n13, OverrideWorldGravity=@n11) {
    bClampLinearTranslationLimitToRefPose=true
    bOverrideWorldGravity=true
    overridePhysicsAsset=&/M_MED_Nucleus_Coat_FrontEnd_Physics
    simulationSpace=BaseBoneSpace
    alphaBoolBlend {blendOption=Linear}
    baseBoneRef {boneName=root}
    componentAppliedLinearAccClamp=<500 500 500>
    overrideWorldGravity {z=150}
   }
   n15 = Get Is Sky Diving() -> (Is Sky Diving:double)
   n16 = "Constraint - dyn_jacketSide_04_l (P)"(ComponentPose=@n14, Alpha=@n15, ConstraintWeights_0=1) {
    alphaBoolBlend {blendOption=Linear}
    boneToModify {boneName=dyn_jacketSide_04_l}
    constraintSetup [{targetBone {boneName=dyn_jacketBack_04_l}}]
    constraintWeights [1]
   }
   n17 = "Constraint - dyn_jacketSide_05_l (P)"(ComponentPose=@n16, Alpha=@n15, ConstraintWeights_0=1) {
    alphaBoolBlend {blendOption=Linear}
    boneToModify {boneName=dyn_jacketSide_05_l}
    constraintSetup [{targetBone {boneName=dyn_jacketBack_05_l}}]
    constraintWeights [1]
   }
   n18 = "Transform (Modify) Bone - Bone: dyn_holster"(ComponentPose=@n17, Alpha=@n13, Translation="0,2,1", Rotation="0,0,0", Scale="1,1,1") {
    translationMode=BMM_Additive
    translationSpace=BCS_BoneSpace
    alphaBoolBlend {blendOption=Linear}
    boneToModify {boneName=dyn_holster}
    translation {y=2, z=1}
   }
   n19 = "Transform (Modify) Bone - Bone: dyn_jacketBack_mstr_r"!(ComponentPose=@n18, Alpha=1, Translation="0.000,0,5") -> (Pose) {
    rotationSpace=BCS_BoneSpace
    translationMode=BMM_Replace
    translationSpace=BCS_BoneSpace
    alphaBoolBlend {blendOption=Linear}
    boneToModify {boneName=dyn_jacketBack_mstr_r}
    rotation=<rot 9.096 4.987 10.829>
    translation {z=5}
   }
   n20 = "Transform (Modify) Bone - Bone: dyn_jacketBack_mstr_l"!(ComponentPose=@n19, Alpha=1, Translation="-0.000,0,5") -> (Pose) {
    rotationSpace=BCS_BoneSpace
    translationMode=BMM_Replace
    translationSpace=BCS_BoneSpace
    alphaBoolBlend {blendOption=Linear}
    boneToModify {boneName=dyn_jacketBack_mstr_l}
    rotation {pitch=5}
    translation {z=5}
   }
   n21 = RigidBody!(ComponentPose=Local To Component(LocalPose=Use cached pose 'body Cache'()), Alpha, ExternalForce) -> (Pose) {
    bClampLinearTranslationLimitToRefPose=true
    bOverrideWorldGravity=true
    overridePhysicsAsset=&M_MED_Nucleus_Coat_FrontEnd_Physics
    simulationSpace=BaseBoneSpace
    alphaBoolBlend {blendOption=Linear}
    baseBoneRef {boneName=root}
    componentAppliedLinearAccClamp=<3000 3000 3000>
    componentLinearAccScale=<0.5 0.4 0.4>
    componentLinearVelScale=<0.5 0.4 0.4>
    overrideWorldGravity {z=-980}
    simSpaceSettings {maxAngularVelocity=3, maxLinearAcceleration=700, maxLinearVelocity=500, velocityScaleZ=0, worldAlpha=1, externalLinearDragV=<2 0.2>}
   }
   n22 = Component To Local(ComponentPose=@n21)
   n23 = RigidBody(ComponentPose=Local To Component(LocalPose=Use cached pose 'body Cache'())) {
    bClampLinearTranslationLimitToRefPose=true
    bOverrideWorldGravity=true
    overridePhysicsAsset=&M_MED_Nucleus_Coat_Physics
    simulationSpace=BaseBoneSpace
    alphaBoolBlend {blendOption=Linear}
    baseBoneRef {boneName=root}
    componentAppliedLinearAccClamp=<3000 3000 3000>
    componentLinearAccScale=<0.5 0.4 0.4>
    componentLinearVelScale=<0.5 0.4 0.4>
    overrideWorldGravity {z=-980}
   }
   n24 = Component To Local(ComponentPose=@n23)
   n25 = Get bIsInFrontEnd()
   n26 = "Transform (Modify) Bone - Bone: dyn_holster"!(
    ComponentPose=Local To Component(LocalPose=Blend Poses by bool(BlendPose_0=@n22, BlendPose_1=@n24, bActiveValue=@n25) {blendType=Linear})
    Alpha=Get Is in Front End()
    Translation="0,2,1"
   ) -> (Pose) {
    translationMode=BMM_Additive
    translationSpace=BCS_BoneSpace
    alphaBoolBlend {blendOption=Linear}
    boneToModify {boneName=dyn_holster}
    scale=<0.9 0.9 0.9>
    translation {y=2, z=1}
   }
   n27 = "Transform (Modify) Bone - Bone: dyn_jacketBack_mstr_r"!(ComponentPose=@n26, Alpha=1) -> (Pose) {
    rotationMode=BMM_Replace
    rotationSpace=BCS_BoneSpace
    translationSpace=BCS_BoneSpace
    alphaBoolBlend {blendOption=Linear}
    alphaScaleBiasClamp {bInterpResult=true, interpSpeedDecreasing=5, interpSpeedIncreasing=7}
    boneToModify {boneName=dyn_jacketBack_mstr_r}
    rotation=<rot -15.0004 -0.002 0.005>
    translation {y=2, z=3}
   }
   n28 = "Transform (Modify) Bone - Bone: dyn_jacketBack_03_r"!(ComponentPose=@n27, Alpha=1) -> (Pose) {
    rotationMode=BMM_Replace
    rotationSpace=BCS_BoneSpace
    translationMode=BMM_Replace
    translationSpace=BCS_BoneSpace
    alphaBoolBlend {blendOption=Linear}
    alphaScaleBiasClamp {bInterpResult=true, interpSpeedDecreasing=5, interpSpeedIncreasing=7}
    boneToModify {boneName=dyn_jacketBack_03_r}
    rotation=<rot 24.9992 -0.0021 0.0016>
    translation=<1.0996 1.081 -0.4205>
   }
   n29 = "Transform (Modify) Bone - Bone: dyn_jacketFront_01_l"!(ComponentPose=@n28, Alpha=1) -> (Pose) {
    rotationMode=BMM_Replace
    rotationSpace=BCS_BoneSpace
    translationSpace=BCS_BoneSpace
    alphaBoolBlend {blendOption=Linear}
    alphaScaleBiasClamp {bInterpResult=true, interpSpeedDecreasing=5, interpSpeedIncreasing=7}
    boneToModify {boneName=dyn_jacketFront_01_l}
    rotation=<rot -5.9434 7.4224 -24.5276>
    translation {y=2, z=3}
   }
   n30 = "Transform (Modify) Bone - Bone: dyn_jacketSide_mstr_l"!(ComponentPose=@n29, Alpha=1) -> (Pose) {
    rotationMode=BMM_Replace
    rotationSpace=BCS_BoneSpace
    translationSpace=BCS_BoneSpace
    alphaBoolBlend {blendOption=Linear}
    alphaScaleBiasClamp {bInterpResult=true, interpSpeedDecreasing=5, interpSpeedIncreasing=7}
    boneToModify {boneName=dyn_jacketSide_mstr_l}
    rotation=<rot -9.9811 5.0583 -0.453>
    translation {y=2, z=3}
   }
   n31 = Blend Poses by bool(BlendPose_0=Component To Local(ComponentPose=@n20), BlendPose_1=Component To Local(ComponentPose=@n30), bActiveValue=Get bIsSkyDiving()) {blendType=Linear}
   n32 = "Transform (Modify) Bone - Bone: dyn_jacketFront_mstr_l"(ComponentPose=Local To Component(LocalPose=Use cached pose 'AnimCache'()), Alpha=1, Translation="0,0,10", Rotation="0,0,0", Scale="1,1,1") {
    rotationMode=BMM_Additive
    rotationSpace=BCS_BoneSpace
    scaleMode=BMM_Additive
    scaleSpace=BCS_BoneSpace
    translationMode=BMM_Additive
    translationSpace=BCS_WorldSpace
    alphaBoolBlend {blendOption=Linear}
    boneToModify {boneName=dyn_jacketFront_mstr_l}
    translation {z=10}
   }
   n33 = "Transform (Modify) Bone - Bone: dyn_jacketFront_02_l"(ComponentPose=@n32, Alpha=1, Translation="0,0,0", Rotation="0,0,0", Scale="1,1,1") {
    rotationMode=BMM_Additive
    rotationSpace=BCS_BoneSpace
    scaleMode=BMM_Additive
    scaleSpace=BCS_BoneSpace
    translationMode=BMM_Additive
    translationSpace=BCS_WorldSpace
    alphaBoolBlend {blendOption=Linear}
    boneToModify {boneName=dyn_jacketFront_02_l}
   }
   n34 = "Transform (Modify) Bone - Bone: dyn_jacketBack_mstr_l"!(ComponentPose=@n33, Alpha=1, Translation="0,0,2", Rotation:Rotator="0,-20,0.000", Scale:Vector="1,1,1") -> (Pose) {
    rotationMode=BMM_Additive
    rotationSpace=BCS_BoneSpace
    scaleMode=BMM_Additive
    scaleSpace=BCS_BoneSpace
    translationMode=BMM_Additive
    translationSpace=BCS_WorldSpace
    alphaBoolBlend {blendOption=Linear}
    boneToModify {boneName=dyn_jacketBack_mstr_l}
    rotation {yaw=-20}
    translation {z=2}
   }
   n35 = "Transform (Modify) Bone - Bone: dyn_jacketBack_mstr_r"!(ComponentPose=@n34, Alpha=1, Translation:Vector="0,0.000,2", Rotation:Rotator="50,-5,0.000", Scale:Vector="1,1,1") -> (Pose) {
    rotationMode=BMM_Additive
    rotationSpace=BCS_BoneSpace
    scaleMode=BMM_Additive
    scaleSpace=BCS_BoneSpace
    translationMode=BMM_Additive
    translationSpace=BCS_WorldSpace
    alphaBoolBlend {blendOption=Linear}
    boneToModify {boneName=dyn_jacketBack_mstr_r}
    rotation {pitch=50, yaw=-5}
    translation {z=2}
   }
   n36 = RigidBody!(ComponentPose=@n35, Alpha, OverrideWorldGravity:Vector="0,0,-980", ExternalForce) -> (Pose) {
    bClampLinearTranslationLimitToRefPose=true
    bOverrideWorldGravity=true
    overridePhysicsAsset=&/M_MED_Nucleus_Crouch_Physics
    simulationSpace=BaseBoneSpace
    alphaBoolBlend {blendOption=Linear}
    baseBoneRef {boneName=root}
    componentAppliedLinearAccClamp=<3000 3000 3000>
    componentLinearAccScale=<0.5 0.5 0.5>
    componentLinearVelScale=<0.5 0.5 0.5>
    overrideWorldGravity {z=-980}
   }
   n37 = Component To Local(ComponentPose=@n36)
   n38 = Get bIsCrouching()
   n39 = Anim Dynamics - Bone: dyn_zipper(
    ComponentPose=Local To Component(LocalPose=Blend Poses by bool(BlendPose_0=@n37, BlendPose_1=@n31, BlendTime_0=0.2, BlendTime_1=0.2, bActiveValue=@n38) {blendType=Linear})
    Alpha=1
   ) {
    angularDampingOverride=0.9
    angularSpringConstant=2000
    bGravityOverrideInSimSpace=true
    bOverrideAngularDamping=true
    bUseGravityOverride=true
    numSolverIterationsPreUpdate=2
    alphaBoolBlend {blendOption=Linear}
    boundBone {boneName=dyn_zipper}
    componentAppliedLinearAccClamp=<3000 3000 3000>
    componentLinearAccScale=<0.01 0.01 0.01>
    componentLinearVelScale=<0.01 0.01 0.01>
    gravityOverride {z=-980}
    physicsBodyDefinitions [{
      sphereCollisionRadius=0
      boundBone {boneName=dyn_zipper}
      boxExtents=<4 5 5>
      constraintSetup {angularLimitsMax=<7 8 0>, angularLimitsMin=<-7 -8 -5>, angularTarget=<0 0 0>}
      localJointOffset=<0.5 0 0>
     }]
    relativeSpaceBone {boneName=spine_05}
   }
   n40 = Anim Dynamics - Bone: dyn_collar_l(ComponentPose=@n39, Alpha=1, ExternalForce="0,0,0") {
    angularSpringConstant=2000
    bGravityOverrideInSimSpace=true
    bUseGravityOverride=true
    alphaBoolBlend {blendOption=Linear}
    boundBone {boneName=dyn_collar_l}
    componentAppliedLinearAccClamp=<3000 3000 3000>
    componentLinearAccScale=<0.08 0.08 0.08>
    componentLinearVelScale=<0.08 0.08 0.08>
    gravityOverride {z=-980}
    physicsBodyDefinitions [{
      sphereCollisionRadius=0
      boundBone {boneName=dyn_collar_l}
      boxExtents=<8 10 10>
      constraintSetup {angularLimitsMax=<5 15 5>, angularLimitsMin=<-5 -15 -5>, angularTarget=<0 0 0>}
      localJointOffset=<1 0 0>
     }]
    relativeSpaceBone {boneName=spine_05}
   }
   n41 = Anim Dynamics - Bone: dyn_collar_r(ComponentPose=@n40, Alpha=1, ExternalForce="0,0,0") {
    angularSpringConstant=2000
    bGravityOverrideInSimSpace=true
    bUseGravityOverride=true
    alphaBoolBlend {blendOption=Linear}
    boundBone {boneName=dyn_collar_r}
    componentAppliedLinearAccClamp=<3000 3000 3000>
    componentLinearAccScale=<0.1 0.1 0.1>
    componentLinearVelScale=<0.1 0.1 0.1>
    gravityOverride {z=-980}
    physicsBodyDefinitions [{
      sphereCollisionRadius=0
      boundBone {boneName=dyn_collar_r}
      boxExtents=<3 5 5>
      constraintSetup {angularLimitsMax=<5 15 5>, angularLimitsMin=<-5 -15 -5>, angularTarget=<0 0 0>}
      localJointOffset=<2 0 0>
     }]
    relativeSpaceBone {boneName=spine_05}
   }
   n42 = "Constraint - dyn_jacketSide_01_r (P)"(ComponentPose=@n41, Alpha=1, ConstraintWeights_0=1) {
    alphaBoolBlend {blendOption=Linear}
    alphaScaleBiasClamp {bClampResult=true, bInterpResult=true, clampMax=0.5, interpSpeedDecreasing=5, interpSpeedIncreasing=7}
    boneToModify {boneName=dyn_jacketSide_01_r}
    constraintSetup [{targetBone {boneName=pelvis}}]
    constraintWeights [1]
   }
   n43 = "Constraint - dyn_jacketBack_01_r (P)"(ComponentPose=@n42, Alpha=1, ConstraintWeights_0=1) {
    alphaBoolBlend {blendOption=Linear}
    alphaScaleBiasClamp {bClampResult=true, bInterpResult=true, clampMax=0.5, interpSpeedDecreasing=5, interpSpeedIncreasing=7}
    boneToModify {boneName=dyn_jacketBack_01_r}
    constraintSetup [{targetBone {boneName=pelvis}}]
    constraintWeights [1]
   }
   n44 = Get Crouch_amount() -> (Crouch_amount:double)
   n45 = "Constraint - dyn_jacketBack_03_r (P)"(ComponentPose=@n43, Alpha=@n44, ConstraintWeights_0=1) {
    alphaBoolBlend {blendOption=Linear}
    boneToModify {boneName=dyn_jacketBack_03_r}
    constraintSetup [{targetBone {boneName=dyn_jacketBack_03_l}}]
    constraintWeights [1]
   }
   n46 = "Constraint - dyn_jacketBack_03_l (P)"(ComponentPose=@n45, Alpha=@n44, ConstraintWeights_0=1) {
    alphaBoolBlend {blendOption=Linear}
    boneToModify {boneName=dyn_jacketBack_03_l}
    constraintSetup [{targetBone {boneName=dyn_jacketBack_03_r}}]
    constraintWeights [1]
   }
   n47 = Get Not Crouch() -> (Not Crouch:bool)
   n48 = "Transform (Modify) Bone - Bone: dyn_jacketBack_03_l"!(ComponentPose=@n46, bAlphaBoolEnabled:bool=@n47, Translation="0,0,0", Rotation="0,5,0", Scale="1,1,1") -> (Pose) {
    alphaInputType=Bool
    rotationMode=BMM_Additive
    rotationSpace=BCS_BoneSpace
    scaleMode=BMM_Additive
    scaleSpace=BCS_BoneSpace
    translationMode=BMM_Additive
    translationSpace=BCS_BoneSpace
    alphaBoolBlend {blendInTime=0.5, blendOption=Linear, blendOutTime=0.1}
    boneToModify {boneName=dyn_jacketBack_03_l}
    rotation {yaw=5}
   }
   n49 = "Transform (Modify) Bone - Bone: dyn_jacketBack_03_r"!(ComponentPose=@n48, bAlphaBoolEnabled:bool=@n47, Translation:Vector="0,0,0", Rotation:Rotator="0,-5,0", Scale:Vector="1,1,1") -> (Pose) {
    alphaInputType=Bool
    rotationMode=BMM_Additive
    rotationSpace=BCS_BoneSpace
    scaleMode=BMM_Additive
    scaleSpace=BCS_BoneSpace
    translationMode=BMM_Additive
    translationSpace=BCS_BoneSpace
    alphaBoolBlend {blendInTime=0.5, blendOption=Linear, blendOutTime=0.1}
    boneToModify {boneName=dyn_jacketBack_03_r}
    rotation {yaw=-5}
   }
   Output Pose(Result=Component To Local(ComponentPose=@n49)) {layerGroup=None}
   n50 = "Transform (Modify) Bone - Bone: dyn_jacketBack_mstr_r"!(ComponentPose=Local To Component(LocalPose=Use cached pose 'AnimCache'()), Alpha=1, Translation="0,2,8") -> (Pose) {
    rotationSpace=BCS_BoneSpace
    translationMode=BMM_Replace
    translationSpace=BCS_BoneSpace
    alphaBoolBlend {blendOption=Linear}
    alphaScaleBiasClamp {interpSpeedDecreasing=5, interpSpeedIncreasing=7}
    boneToModify {boneName=dyn_jacketBack_mstr_r}
    rotation=<rot 9.096 4.987 10.829>
    translation {y=2, z=8}
   }
   n51 = "Transform (Modify) Bone - Bone: dyn_jacketBack_mstr_l"!(ComponentPose=@n50, Alpha=1, Translation="-0.000,2,3") -> (Pose) {
    rotationSpace=BCS_BoneSpace
    translationMode=BMM_Replace
    translationSpace=BCS_BoneSpace
    alphaBoolBlend {blendOption=Linear}
    alphaScaleBiasClamp {bInterpResult=true, interpSpeedDecreasing=5, interpSpeedIncreasing=7}
    boneToModify {boneName=dyn_jacketBack_mstr_l}
    rotation {pitch=5}
    translation {y=2, z=3}
   }
   n52 = "Transform (Modify) Bone - Bone: dyn_jacketSide_mstr_l"!(ComponentPose=@n51, Alpha=1, Translation:Vector="0.000,-3,5") -> (Pose) {
    rotationSpace=BCS_BoneSpace
    translationMode=BMM_Replace
    translationSpace=BCS_BoneSpace
    alphaBoolBlend {blendOption=Linear}
    alphaScaleBiasClamp {bInterpResult=true, interpSpeedDecreasing=5, interpSpeedIncreasing=7}
    boneToModify {boneName=dyn_jacketSide_mstr_l}
    rotation=<rot 9.096 4.987 10.829>
    translation {y=-3, z=5}
   }
   n53 = "Transform (Modify) Bone - Bone: dyn_jacketSide_mstr_r"!(ComponentPose=@n52, Alpha=1, Translation="-0.000,3,5") -> (Pose) {
    rotationSpace=BCS_BoneSpace
    translationMode=BMM_Replace
    translationSpace=BCS_BoneSpace
    alphaBoolBlend {blendOption=Linear}
    alphaScaleBiasClamp {bInterpResult=true, interpSpeedDecreasing=5, interpSpeedIncreasing=7}
    boneToModify {boneName=dyn_jacketSide_mstr_r}
    rotation {pitch=5}
    translation {y=3, z=5}
   }
   n54 = Get thigh_rot_inr_r() -> (thigh_rot_inr_r:double)
   n55 = Look At - Bone: dyn_jacketFront_mstr_r(ComponentPose=@n53, Alpha=@n54) {
    alphaBoolBlend {blendOption=Linear}
    alphaScaleBiasClamp {bInterpResult=true}
    boneToModify {boneName=dyn_jacketFront_mstr_r}
    lookAtLocation=<0 -12 5>
    lookAtTarget {boneReference {boneName=calf_r}}
    lookAt_Axis {axis=<1 0>}
    lookUp_Axis {axis {x=0}}
   }
   n56 = Look At - Bone: dyn_jacketSide_mstr_r(ComponentPose=@n55, Alpha=@n54) {
    alphaBoolBlend {blendOption=Linear}
    alphaScaleBiasClamp {bInterpResult=true}
    boneToModify {boneName=dyn_jacketSide_mstr_r}
    lookAtLocation {x=0, z=11}
    lookAtTarget {boneReference {boneName=calf_r}}
    lookAt_Axis {axis=<1 0>}
    lookUp_Axis {axis {x=0}}
   }
   n57 = Get thigh_rot_up_r() -> (thigh_rot_up_r:double)
   n58 = "Transform (Modify) Bone - Bone: dyn_jacketFront_01_r"(ComponentPose=@n56, Alpha=@n57, Translation="0,0,0", Rotation="9.096,4.987,10.829", Scale="1,1,1") {
    rotationMode=BMM_Additive
    rotationSpace=BCS_BoneSpace
    translationMode=BMM_Additive
    translationSpace=BCS_BoneSpace
    alphaBoolBlend {blendOption=Linear}
    boneToModify {boneName=dyn_jacketFront_01_r}
    rotation=<rot 9.096 4.987 10.829>
   }
   n59 = "Transform (Modify) Bone - Bone: dyn_jacketSide_01_r"(ComponentPose=@n58, Alpha=@n57, Translation="0,0,0", Rotation="5.000,0.000,0.000", Scale="1,1,1") {
    rotationMode=BMM_Additive
    rotationSpace=BCS_BoneSpace
    translationMode=BMM_Additive
    translationSpace=BCS_BoneSpace
    alphaBoolBlend {blendOption=Linear}
    boneToModify {boneName=dyn_jacketSide_01_r}
    rotation {pitch=5}
   }
   n60 = Get thigh_rot_inr_l() -> (thigh_rot_inr_l:double)
   n61 = Look At - Bone: dyn_jacketFront_mstr_l(ComponentPose=@n59, Alpha=@n60) {
    alphaBoolBlend {blendOption=Linear}
    alphaScaleBiasClamp {bInterpResult=true}
    boneToModify {boneName=dyn_jacketFront_mstr_l}
    lookAtLocation=<0 12 -5>
    lookAtTarget {boneReference {boneName=calf_l}}
    lookAt_Axis {axis=<1 0>}
    lookUp_Axis {axis {x=0}}
   }
   n62 = Look At - Bone: dyn_jacketSide_mstr_l(ComponentPose=@n61, Alpha=@n60) {
    alphaBoolBlend {blendOption=Linear}
    alphaScaleBiasClamp {bInterpResult=true}
    boneToModify {boneName=dyn_jacketSide_mstr_l}
    lookAtLocation {x=0, z=-11}
    lookAtTarget {boneReference {boneName=calf_l}}
    lookAt_Axis {axis=<1 0>}
    lookUp_Axis {axis {x=0}}
   }
   n63 = Get thigh_rot_up_l() -> (thigh_rot_up_l:double)
   n64 = "Transform (Modify) Bone - Bone: dyn_jacketFront_01_l"!(ComponentPose=@n62, Alpha=@n63, Translation:Vector="0,0,0", Rotation:Rotator="9.096,-4.987,-10.828", Scale:Vector="1,1,1") -> (Pose) {
    rotationMode=BMM_Additive
    rotationSpace=BCS_BoneSpace
    translationMode=BMM_Additive
    translationSpace=BCS_BoneSpace
    alphaBoolBlend {blendOption=Linear}
    boneToModify {boneName=dyn_jacketFront_01_l}
    rotation=<rot 9.096 -4.987 -10.8284>
   }
   n65 = "Transform (Modify) Bone - Bone: dyn_jacketSide_01_l"(ComponentPose=@n64, Alpha=@n63, Translation="0,0,0", Rotation="5,0,0", Scale="1,1,1") {
    rotationMode=BMM_Additive
    rotationSpace=BCS_BoneSpace
    translationMode=BMM_Additive
    translationSpace=BCS_BoneSpace
    alphaBoolBlend {blendOption=Linear}
    boneToModify {boneName=dyn_jacketSide_01_l}
    rotation {pitch=5}
   }
   Save cached pose 'body Cache'(Pose=Component To Local(ComponentPose=@n65)) {cacheName=body Cache}
   n66 = "Transform (Modify) Bone - Bone: dyn_ground_root"(
    ComponentPose=Local To Component(LocalPose=Copy Pose From Mesh() {bUseAttachedParent=true})
    Alpha=Get PlummetAlpha() -> (PlummetAlpha:double)
    Translation="0,0,-600"
    Rotation="0,0,0"
    Scale="1,1,1"
   ) {
    translationMode=BMM_Additive
    translationSpace=BCS_BoneSpace
    alphaBoolBlend {blendOption=Linear}
    boneToModify {boneName=dyn_ground_root}
    translation {z=-600}
   }
   n67 = Control Rig(Source=Component To Local(ComponentPose=@n66)) {
    alphaBoolBlend {blendOption=Linear}
    controlRigAssetReference {blueprintRigClass=&/M_MED_Nucleus_Deform_CtrlRig_twists.M_MED_Nucleus_Deform_CtrlRig_twists_C}
   }
   n68 = Control Rig!(
    Source=@n67
    Alpha
    KneeRot_fr_l_114deg:Vector
    KneeTrans_fr_l_114deg:Vector="0.258,-3.591,0.009"
    KneeScale_fr_l_114deg:Vector="1,1,1"
    KneeScale_fr_l_130deg:Vector="1,1,1"
    KneeRot_fr_l_130deg:Vector="0,0,-51.2"
    KneeTrans_fr_l_130deg:Vector="0.258,-3.591,0.009"
   ) -> (Pose) {
    alphaBoolBlend {blendOption=Linear}
    controlRigAssetReference {blueprintRigClass=&/M_MED_Nucleus_Deform_CtrlRig_sculpters1.M_MED_Nucleus_Deform_CtrlRig_sculpters1_C}
   }
   n69 = "Transform (Modify) Bone - Bone: dyn_root_simspace"(
    ComponentPose=Local To Component(LocalPose=@n68)
    Alpha=1
    Translation="0,0,0"
    Rotation=Get sim_space_rotation() -> (sim_space_rotation:Rotator)
    Scale="1,1,1"
   ) {rotationMode=BMM_Replace, rotationSpace=BCS_WorldSpace, alphaBoolBlend {blendOption=Linear}, boneToModify {boneName=dyn_root_simspace}}
   n70 = "Transform (Modify) Bone - Bone: dyn_jacketFront_mstr_r"(ComponentPose=@n69, Alpha=1, Translation="0.6,0.000,1", Rotation="2,0.000,0.000", Scale="1,1,1") {
    rotationMode=BMM_Additive
    rotationSpace=BCS_BoneSpace
    scaleMode=BMM_Additive
    scaleSpace=BCS_BoneSpace
    translationMode=BMM_Additive
    translationSpace=BCS_BoneSpace
    alphaBoolBlend {blendOption=Linear}
    boneToModify {boneName=dyn_jacketFront_mstr_r}
    rotation {pitch=2}
    translation {x=0.6, z=1}
   }
   n71 = "Transform (Modify) Bone - Bone: dyn_jacketSide_mstr_r"!(ComponentPose=@n70, Alpha=1, Translation="0.5,0.000,0.5", Rotation:Rotator="1,0.000,0.000", Scale:Vector="1,1,1") -> (Pose) {
    rotationMode=BMM_Additive
    rotationSpace=BCS_BoneSpace
    scaleMode=BMM_Additive
    scaleSpace=BCS_BoneSpace
    translationMode=BMM_Additive
    translationSpace=BCS_BoneSpace
    alphaBoolBlend {blendOption=Linear}
    boneToModify {boneName=dyn_jacketSide_mstr_r}
    rotation {pitch=1}
    translation {x=0.5, z=0.5}
   }
   n72 = "Transform (Modify) Bone - Bone: dyn_collar_r"(ComponentPose=@n71, Alpha=Get upperarm_r() -> (upperarm_r:double), Translation="0,0,-5", Rotation="-60,0,10", Scale="1,1,1") {
    rotationMode=BMM_Additive
    rotationSpace=BCS_BoneSpace
    scaleMode=BMM_Additive
    scaleSpace=BCS_BoneSpace
    translationMode=BMM_Additive
    translationSpace=BCS_BoneSpace
    alphaBoolBlend {blendOption=Linear}
    alphaScaleBiasClamp {bMapRange=true, inRange {max=0, min=-20}, outRange {max=0, min=1}}
    boneToModify {boneName=dyn_collar_r}
    rotation {pitch=-60, roll=10}
    translation {z=-5}
   }
   n73 = "Transform (Modify) Bone - Bone: dyn_collar_l"(ComponentPose=@n72, Alpha=Get upperarm_l() -> (upperarm_l:double), Translation="0,0,5", Rotation="0,0,0", Scale="1,1,1") {
    rotationMode=BMM_Additive
    scaleMode=BMM_Additive
    scaleSpace=BCS_BoneSpace
    translationMode=BMM_Additive
    translationSpace=BCS_BoneSpace
    alphaBoolBlend {blendOption=Linear}
    alphaScaleBiasClamp {bMapRange=true, inRange {max=13}}
    boneToModify {boneName=dyn_collar_l}
    translation {z=5}
   }
   n74 = "Transform (Modify) Bone - Bone: dyn_holster"!(ComponentPose=@n73, Alpha=1, Translation="0,2,1", Rotation:Rotator="0,0,0", Scale:Vector="0.9,0.9,0.9") -> (Pose) {
    scaleMode=BMM_Additive
    translationMode=BMM_Additive
    translationSpace=BCS_BoneSpace
    alphaBoolBlend {blendOption=Linear}
    boneToModify {boneName=dyn_holster}
    scale=<0.9 0.9 0.9>
    translation {y=2, z=1}
   }
   n75 = "Transform (Modify) Bone - Bone: upperarm_twist_01_r"(ComponentPose=@n74, Alpha=Get upperarm_r(), Translation="0,-3,0", Rotation="15,-10,10", Scale="1,1,1") {
    rotationMode=BMM_Additive
    rotationSpace=BCS_BoneSpace
    translationMode=BMM_Additive
    translationSpace=BCS_BoneSpace
    alphaBoolBlend {blendOption=Linear}
    alphaScaleBiasClamp {bMapRange=true, inRange {max=0, min=-14}, outRange {max=0, min=1}}
    boneToModify {boneName=upperarm_twist_01_r}
    rotation=<rot 15 -10 10>
    translation {y=-3}
   }
   Save cached pose 'AnimCache'(Pose=Component To Local(ComponentPose=@n75)) {cacheName=AnimCache}
  ]}
 EventGraph {nodes [
   Event BlueprintUpdateAnimation()
   > n1 = Sequence!(execute) -> (
    then_0
    then_1
    then_2:Exec
    then_3:Exec
    then_4:Exec
    then_5:Exec
    then_6:Exec
    then_7:Exec
    then_8:Exec
    then_9:Exec
    then_10:Exec
    then_11:Exec
   )
   Branch(execute=@n1.then_1, Condition=OR Boolean(A=Get bIsSkyDiving(), B=Get bIsInFrontEnd()))
   > Set frontEndorSkydiving(frontEndorSkydiving:bool=true) -> (Output_Get:bool)
   n2 = PlayerMovement(execute=@n1.then_10, AnimInstance=Self-Reference(), XDotProductCoefficient=1, YDotProductCoefficient=1, ZDotProductCoefficient=1)
   n3 = MapRangeClamped(Value=@n2.YMovement, InRangeA=0, InRangeB=-1)
   n4 = float * float(A="Absolute (Float)"(A=@n2), B=0.5)
   > Set BwdMovement(BwdMovement:double=MapRangeClamped(
     Value=float + float(A=@n4, B=@n3) {is=title K2Node_CommutativeAssociativeBinaryOperator}
     InRangeA=0
     InRangeB=1.1
     OutRangeA=0.4
     OutRangeB=0
    )) -> (Output_Get:double)
   n5 = MapRangeClamped(Value=@n2, InRangeA=-0.2, InRangeB=1)
   n6 = float * float(A="Absolute (Float)"(A=@n2.YMovement), B=0.2)
   > Set RightMovement(RightMovement:double=MapRangeClamped(Value=float + float(A=@n6, B=@n5), InRangeA=0, InRangeB=1.2, OutRangeB=0.5)) -> (Output_Get:double)
   n7 = MapRangeClamped(Value=@n2, InRangeA=0.2, InRangeB=-1, OutRangeB=0.3)
   n8 = float * float(A="Absolute (Float)"(A=@n2.YMovement), B=0.2)
   > Set LeftMovement(LeftMovement:double=MapRangeClamped(Value=float + float(A=@n8, B=@n7), InRangeA=0, InRangeB=1.2, OutRangeB=0.5)) -> (Output_Get:double)
   n9 = Get bIsInFrontEnd()
   n10 = Get bIsPlayingEmote()
   > Set Emote Off(Emote Off:double=SelectFloat(A=0, B=1, bPickA=AND Boolean(A=@n9, B=@n10))) -> (Output_Get:double)
   n11 = Get Crouching() -> (Crouching:bool)
   > Set Crouch_amount(Crouch_amount:double=SelectFloat(A=0.3, B=0.3, bPickA=NOT Boolean(A=@n11))) -> (Output_Get:double)
   > Set Not Crouch(Not Crouch:bool=NOT Boolean(A=Get Crouching())) -> (Output_Get:bool)
   Get!(execute:Exec=@n1.then_11) -> (then:Exec, else:Exec, Anim_Inst:Anim Instance Object Reference)
   n12 = GetCurveValue(self=Get Anim_Inst() -> (Anim_Inst:Anim Instance Object Reference), CurveName=PhysicsGroundOffset)
   > n13 = Branch(Condition="float > float"(A=@n12))
   > Set PlummetAlpha(PlummetAlpha:double=@n12) -> (Output_Get:double)
   Set PlummetAlpha(execute=@n13.else, PlummetAlpha=0)
   n14 = Branch(execute=@n1.then_2, Condition=OR Boolean(A=Get bIsJumping(), B=Get bIsFalling()))
   > Set JumpingFalling(JumpingFalling:bool=true) -> (Output_Get:bool)
   Set JumpingFalling(execute=@n14.else, JumpingFalling=false)
   n15 = Break Fort Anim Input Slope Info!(FortAnimInput_SlopeInfo:"Fort Anim Input Slope Info Structure (by ref)"=Get SlopeInfoInput() -> (SlopeInfoInput:Fort Anim Input Slope Info Structure)) -> (RootSlopeTranslation:Vector=~, LeftFootSlopeTranslation:Vector=~, RightFootSlopeTranslation:Vector=~, SlopeRotation:Rotator=~) {is=title K2Node_BreakStruct}
   Set slope_root_trans(execute=@n1.then_3, slope_root_trans:Vector=@n15) -> (Output_Get:Vector)
   > Set slope_foot_rot(slope_foot_rot:Rotator=@n15.SlopeRotation) -> (Output_Get:Rotator)
   n16 = Gravity Override State Function DEPRECATED(
    execute=@n1.then_4
    CharacterPartAnimInstance=Self-Reference()
    GameTimeInSeconds=GetGameTimeInSeconds()
    GravityOverrideParams=Get Gravity Override Params()
   )
   > Set GravityOverride(GravityOverride=@n16)
   n17 = ComplexSimSpaceFunction DEPRECATED!(
    execute=@n1.then_5
    CharacterPartAnimInstance:Custom Character Part Anim Instance Object Reference=Self-Reference()
    InterpSpeedMax:double=9
    PawnSpeedMax:double=20
    RInterpErrorTolerance:double=1
    DeltaTime:double=0
    DebugDraw:bool=false
    sim_space_rotation_in:Rotator=Get sim_space_rotation()
    inComponentSpaceIn:bool=Get InComponentSpace() -> (InComponentSpace:bool)
    inComponentSpaceLastFrameIn:bool=Get InComponentSpaceLastFrame() -> (InComponentSpaceLastFrame:bool)
    lockedOnSimSpaceIn:bool=Get LockedOnSimSpace() -> (LockedOnSimSpace:bool)
    InterpSpeedIn:double=Get InterpSpeed() -> (InterpSpeed:double)
   ) -> (
    then
    sim_space_rotation:Rotator
    inComponentSpaceOut:bool
    inComponentSpaceLastFrameOut:bool
    lockedOnSimSpaceOut:bool
    InterpSpeedOut:double
   )
   > Set sim_space_rotation(sim_space_rotation:Rotator=@n17) -> (Output_Get:Rotator)
   > Set InComponentSpace(InComponentSpace:bool=@n17.inComponentSpaceOut) -> (Output_Get:bool)
   > Set InComponentSpaceLastFrame(InComponentSpaceLastFrame:bool=@n17.inComponentSpaceLastFrameOut) -> (Output_Get:bool)
   > Set LockedOnSimSpace(LockedOnSimSpace:bool=@n17.lockedOnSimSpaceOut) -> (Output_Get:bool)
   > Set InterpSpeed(InterpSpeed:double=@n17.InterpSpeedOut) -> (Output_Get:double)
   n18 = Branch(execute=@n1.then_7, Condition=Get bIsInFrontEnd())
   > Set Is in Front End(Is in Front End:double=1) -> (Output_Get:double)
   > Set Is not Front End(Is not Front End:double=0) -> (Output_Get:double)
   Set Is in Front End(execute=@n18.else, Is in Front End=0)
   > Set Is not Front End(Is not Front End=1)
   n19 = Branch(execute=@n1.then_8, Condition=Get bIsSkyDiving())
   > Set Is Sky Diving(Is Sky Diving:double=1) -> (Output_Get:double)
   Set Is Sky Diving(execute=@n19.else, Is Sky Diving=0)
   n20 = GetOwnerSkeletalMeshComponent()
   n21 = BreakRotator(InRot=BreakTransform(InTransform=GetDeltaTransformFromRefPose(self=@n20, BoneName=thigh_r, BaseName=pelvis)).Rotation)
   Set thigh_rot_up_r(execute=@n1.then_9, thigh_rot_up_r:double=MapRangeClamped(Value=@n21.Yaw, InRangeA=-15, InRangeB=-100)) -> (Output_Get:double)
   > Set thigh_rot_inr_r(thigh_rot_inr_r:double=MapRangeClamped(Value=@n21.Pitch, InRangeA=0, InRangeB=-10, OutRangeA=1, OutRangeB=0)) -> (Output_Get:double)
   > Set thigh_rot_otr_r(thigh_rot_otr_r:double=MapRangeClamped(Value=@n21.Pitch, InRangeA=20, InRangeB=50)) -> (Output_Get:double)
   n22 = GetOwnerSkeletalMeshComponent()
   n23 = BreakRotator(InRot=BreakTransform(InTransform=GetDeltaTransformFromRefPose(self=@n22, BoneName=thigh_l, BaseName=pelvis)).Rotation)
   > Set thigh_rot_up_l(thigh_rot_up_l:double=MapRangeClamped(Value=@n23.Yaw, InRangeA=-10, InRangeB=-100)) -> (Output_Get:double)
   > Set thigh_rot_inr_l(thigh_rot_inr_l:double=MapRangeClamped(Value=@n23.Pitch, InRangeA=0, InRangeB=-10, OutRangeA=1, OutRangeB=0)) -> (Output_Get:double)
   > Set thigh_rot_otr_l(thigh_rot_otr_l:double=MapRangeClamped(Value=@n23.Pitch, InRangeA=20, InRangeB=50)) -> (Output_Get:double)
   Set upperarm_r(
    execute=@n1.then_0
    upperarm_r:double=GetAlphaFromReferenceTransformDelta(SkeletalMeshComponent=Get Skel_Mesh(), BoneName=thigh_r, BaseName=pelvis, InputRange=(X=0,Y=-90), TwistAxis="0,0,1")
   ) -> (Output_Get:double)
   > Set thigh_rot_l(thigh_rot_l=GetAlphaFromReferenceTransformDelta(SkeletalMeshComponent=Get Skel_Mesh(), BoneName=thigh_l, BaseName=pelvis, InputRange=(X=-20,Y=-101), TwistAxis="0,0,1"))
   n24 = GetOwnerSkeletalMeshComponent()
   n25 = GetDeltaTransformFromRefPose(self=@n24, BoneName=upperarm_r, BaseName=clavicle_r)
   > Set upperarm_r(upperarm_r=BreakRotator(InRot=BreakTransform(InTransform=@n25).Rotation))
   n26 = GetDeltaTransformFromRefPose(self=@n24, BoneName=clavicle_l, BaseName=spine_05)
   > Set upperarm_l(upperarm_l:double=BreakRotator(InRot=BreakTransform(InTransform=@n26).Rotation)) -> (Output_Get:double)
   Event OnBeginCrouching()
   > Set Crouching(Crouching:bool=true) -> (Output_Get:bool)
   Event OnEndCrouching()
   > Set Crouching(Crouching=false)
   n27 = GetOwnerSkeletalMeshComponent()
   Is Valid(InputObject=@n27)
   > Set Skel_Mesh(Skel_Mesh=@n27)
   > n28 = Cast To FortPlayerPawn(Object=GetOwningActor())
   > n29 = Set Pawn(Pawn:Fort Player Pawn Object Reference=@n28) -> (Output_Get:Fort Player Pawn Object Reference)
   > Set Anim_Inst(Anim_Inst:Anim Instance Object Reference=GetAnimInstance(self=Get Mesh(self=@n29))) -> (Output_Get:Anim Instance Object Reference)
  ]}
]
variables [
 Anim_Inst {type=AnimInstance, default=null}
 BwdMovement {type=number}
 Crouch_amount {type=number}
 Crouching {type=boolean}
 Emote Off {type=number}
 frontEndorSkydiving {type=boolean}
 Gravity Override Params {type=GravityOverrideParamsStruct}
 GravityOverride {type=Vector}
 InComponentSpace {type=boolean}
 InComponentSpaceLastFrame {type=boolean}
 InterpSpeed {type=number}
 Is in Front End {type=number}
 Is not Front End {type=number}
 Is Sky Diving {type=number}
 JumpingFalling {type=boolean}
 LeftMovement {type=number}
 LockedOnSimSpace {type=boolean, default=true}
 Not Crouch {type=boolean}
 Pawn {type=FortPlayerPawn}
 PlummetAlpha {type=number}
 RightMovement {type=number}
 sim_space_rotation {type=Rotator}
 Skel_Mesh {type=SkeletalMeshComponent}
 slope_foot_rot {type=Rotator}
 slope_root_trans {type=Vector}
 thigh_rot_inr_l {type=number}
 thigh_rot_inr_r {type=number}
 thigh_rot_l {type=number}
 thigh_rot_otr_l {type=number}
 thigh_rot_otr_r {type=number}
 thigh_rot_up_l {type=number}
 thigh_rot_up_r {type=number}
 upperarm_l {type=number}
 upperarm_r {type=number}
]
